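# Control Rig for the F_MED_MillionaireCowgirl character body that procedurally corrects coat, lapel and sleeve bones as the limbs move.
# Spherical pose readers on thighs, calves and upper arms store weights in variables, which drive additive-style corrective bone transforms, with one left-leg pose gated to emotes.
# A final pass plane-clamps lapels and right sleeve against body controls, then adjusts the right lapel by its rotation.
# Several node chains, including run correctives, are disconnected.
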ControlRigBlueprint /BRCosmetics/Characters/Player/Female/Medium/Bodies/F_MED_MillionaireCowgirl/Meshes/F_MED_MillionaireCowgirl_CtrlRig
graphs [
 RigVMFunctionLibrary {links [], nodes []}
 RigVMFunction_Sequence_1_ContainedGraph {nodes [
   n1 = RigVMFunctionEntryNode()
   RigVMFunction_Sequence = &RigVMFunction_Sequence(ExecuteContext=@n1.ExecuteContext)
   n2 = &RigVMFunction_Sequence(ExecuteContext=@RigVMFunction_Sequence.B)
   n3 = &RigVMFunction_Sequence(ExecuteContext=@n2.B)
   RigVMFunctionReturnNode(A=@RigVMFunction_Sequence.A, B=@n2.A, C=@n3.A, D=@n3.B)
  ]}
 RigVMModel {nodes [
   &RigUnit_ModifyTransforms(
    ExecutePin=()
    ItemToModify="((Item=(Type=Bone,Name=\"dyn_lapel_l_02\"),Transform=(Rotation=(X=0,Y=0,Z=0.2588,W=0.9659),Translation=(X=0,Y=0,Z=0),Scale3D=(X=1,Y=1,Z=1))),(Item=(Type=Bone,Name=\"dyn_lapel_l_01\"),Transform=(Rotation=(X=0,Y=0,Z=0,W=1),Translation=(X=0,Y=0,Z=0),Scale3D=(X=1, ...(truncated)"
    WeightMinimum=0
    Weight=RigVMVariableNode() {variable=L Arm Out Up}.Value
   )
   > &RigUnit_ModifyTransforms(
    ItemToModify="((Item=(Type=Bone,Name=\"dyn_lapel_r_02\"),Transform=(Rotation=(X=0.0226,Y=0.2578,Z=-0.0842,W=0.9623),Translation=(X=0,Y=0,Z=0),Scale3D=(X=1,Y=1,Z=1))),(Item=(Type=Bone,Name=\"dyn_lapel_r_01\"),Transform=(Rotation=(X=0,Y=0,Z=0,W=1),Translation=(X=0,Y=0,Z=0),Sc ...(truncated)"
    WeightMinimum=0
    Weight=RigVMVariableNode() {variable=R Arm Out Up}.Value
   )
   Multiply = &RigVMFunction_MathDoubleMul(A=RigVMVariableNode() {variable=R Thigh Run}.Value, B=RigVMVariableNode() {variable=R Calf Run}.Value)
   &RigUnit_ModifyTransforms(
    ExecutePin=()
    ItemToModify="((Item=(Type=Bone,Name=\"dyn_coat_side_r_03\"),Transform=(Rotation=(X=-0.0038,Y=-0.0871,Z=0.0435,W=0.9952),Translation=(X=0,Y=0,Z=0),Scale3D=(X=1,Y=1,Z=1))),(Item=(Type=Bone,Name=\"dyn_coat_bck_r_base\"),Transform=(Rotation=(X=-0.0151,Y=-0.1723,Z=0.1723,W=0.96 ...(truncated)"
    Weight=@Multiply.Result
   )
   > &RigUnit_ModifyTransforms(
    ItemToModify="((Item=(Type=Bone,Name=\"dyn_coat_bck_l_base\"),Transform=(Rotation=(X=0.0151,Y=-0.0858,Z=-0.173,W=0.9811),Translation=(X=0,Y=0,Z=0),Scale3D=(X=1,Y=1,Z=1))),(Item=(Type=Bone,Name=\"dyn_coat_bck_l_02\"),Transform=(Rotation=(X=0,Y=-0.1736,Z=0,W=0.9848),Translati ...(truncated)"
    Weight=&RigVMFunction_MathDoubleMul(A=RigVMVariableNode() {variable=L Thigh Run}.Value, B=RigVMVariableNode() {variable=L Calf Run}.Value).Result
   )
   &RigUnit_ModifyTransforms(
    ExecutePin=()
    ItemToModify="((Item=(Type=Bone,Name=\"dyn_coat_fr_r_base\"),Transform=(Rotation=(X=-0.2395,Y=-0.1578,Z=0.3936,W=0.8734),Translation=(X=-4,Y=-5,Z=2),Scale3D=(X=1,Y=1,Z=1))),(Item=(Type=Bone,Name=\"dyn_coat_side_r_base\"),Transform=(Rotation=(X=-0.0027,Y=-0.1717,Z=0.0934,W=0 ...(truncated)"
    Weight=RigVMVariableNode() {variable=R Leg Right}.Value
   )
   > &RigUnit_ModifyTransforms(
    ItemToModify="((Item=(Type=Bone,Name=\"dyn_coat_fr_r_base\"),Transform=(Rotation=(X=-0.1246,Y=0.0304,Z=0.255,W=0.9584),Translation=(X=-11,Y=10.108,Z=3.9723),Scale3D=(X=1,Y=1,Z=1))),(Item=(Type=Bone,Name=\"dyn_coat_side_r_base\"),Transform=(Rotation=(X=0,Y=0,Z=0,W=1),Transla ...(truncated)"
    Weight=RigVMVariableNode() {variable=R Leg Fwd}.Value
   )
   > &RigUnit_ModifyTransforms(
    ItemToModify="((Item=(Type=Bone,Name=\"dyn_coat_side_r_base\"),Transform=(Rotation=(X=0,Y=-0.1305,Z=0,W=0.9914),Translation=(X=0,Y=0,Z=0),Scale3D=(X=1,Y=1,Z=1))),(Item=(Type=Bone,Name=\"dyn_coat_bck_r_base\"),Transform=(Rotation=(X=0,Y=-0.342,Z=0,W=0.9397),Translation=(X=0, ...(truncated)"
    Weight=RigVMVariableNode() {variable=R Leg Back}.Value
   )
   > &RigUnit_ModifyTransforms(
    ItemToModify="((Item=(Type=Bone,Name=\"dyn_coat_fr_l_base\"),Transform=(Rotation=(X=0.1297,Y=-0.0508,Z=-0.0063,W=0.9902),Translation=(X=-16.1793,Y=-10.7617,Z=-0.809),Scale3D=(X=1,Y=1,Z=1))),(Item=(Type=Bone,Name=\"dyn_coat_fr_l_02\"),Transform=(Rotation=(X=0,Y=0,Z=0,W=1),Tr ...(truncated)"
    Weight=RigVMVariableNode() {variable=L Leg Fwd}.Value
   )
   > &RigUnit_ModifyTransforms(
    ItemToModify="((Item=(Type=Bone,Name=\"dyn_coat_fr_l_base\"),Transform=(Rotation=(X=0.1573,Y=-0.2039,Z=-0.1957,W=0.9462),Translation=(X=-1.3112,Y=0.5263,Z=3.1937),Scale3D=(X=1,Y=1,Z=1))),(Item=(Type=Bone,Name=\"dyn_coat_fr_l_02\"),Transform=(Rotation=(X=0.0038,Y=0.0871,Z=0. ...(truncated)"
    Weight=RigVMVariableNode() {variable=L Leg Left}.Value
   )
   > &RigUnit_ModifyTransforms(
    ItemToModify="((Item=(Type=Bone,Name=\"dyn_coat_bck_l_base\"),Transform=(Rotation=(X=-0.0868,Y=-0.2578,Z=-0.0076,W=0.9623),Translation=(X=0,Y=0,Z=0),Scale3D=(X=1,Y=1,Z=1))),(Item=(Type=Bone,Name=\"dyn_coat_side_l_base\"),Transform=(Rotation=(X=-0.0076,Y=-0.0867,Z=-0.0433,W= ...(truncated)"
    Weight=RigVMVariableNode() {variable=L Leg Back}.Value
   )
   > &RigUnit_ModifyTransforms(
    ItemToModify="((Item=(Type=Bone,Name=\"dyn_sleeve_r_fr\"),Transform=(Rotation=(X=0,Y=0,Z=0,W=1),Translation=(X=0,Y=0,Z=-1.9187),Scale3D=(X=1,Y=1,Z=1))),(Item=(Type=Bone,Name=\"dyn_lapel_r_02\"),Transform=(Rotation=(X=0,Y=0,Z=0,W=1),Translation=(X=0,Y=0,Z=0),Scale3D=(X=1,Y=1 ...(truncated)"
    WeightMinimum=0
    Weight=RigVMVariableNode() {variable=R Arm Up_Full}.Value
   )
   > n1 = &RigUnit_ModifyTransforms(
    ItemToModify="((Item=(Type=Bone,Name=\"dyn_lapel_r_01\"),Transform=(Rotation=(X=-0.083,Y=-0.0905,Z=0.0357,W=0.9918),Translation=(X=0,Y=0.2,Z=0.5),Scale3D=(X=1,Y=1,Z=1))),(Item=(Type=Bone,Name=\"dyn_sleeve_r_fr\"),Transform=(Rotation=(X=0,Y=0.0872,Z=0,W=0.9962),Translation=( ...(truncated)"
    WeightMinimum=0
    Weight=RigVMVariableNode() {variable=R Arm Back}.Value
   )
   n2 = &RigVMFunction_ControlFlowBranch(ExecuteContext=@n1.ExecutePin, Condition=RigVMVariableNode() {variable=Is Playing Emote}.Value)
   &RigUnit_ModifyTransforms(
    ItemToModify="((Item=(Type=Bone,Name=\"dyn_coat_fr_l_base\"),Transform=(Rotation=(X=-0.017,Y=-0.1294,Z=-0.1294,W=0.983),Translation=(X=0,Y=0,Z=3),Scale3D=(X=1,Y=1,Z=1))),(Item=(Type=Bone,Name=\"dyn_coat_fr_l_01\"),Transform=(Rotation=(X=0,Y=0,Z=0,W=1),Translation=(X=0,Y=0,Z ...(truncated)"
    ExecutePin=@n2.True
    Weight=RigVMVariableNode() {variable=L Leg Out}.Value
   )
   &RigUnit_ModifyTransforms(
    ExecutePin=()
    ItemToModify="((Item=(Type=Bone,Name=\"dyn_lapel_l_02\"),Transform=(Rotation=(X=0,Y=0,Z=0.2588,W=0.9659),Translation=(X=0,Y=0,Z=0),Scale3D=(X=1,Y=1,Z=1))),(Item=(Type=Bone,Name=\"dyn_lapel_l_01\"),Transform=(Rotation=(X=0,Y=0,Z=0,W=1),Translation=(X=0,Y=0,Z=0),Scale3D=(X=1, ...(truncated)"
    WeightMinimum=0
    Weight=RigVMVariableNode() {variable=L Arm Out Up}.Value
   )
   > &RigUnit_ModifyTransforms(
    ItemToModify="((Item=(Type=Bone,Name=\"dyn_lapel_r_02\"),Transform=(Rotation=(X=0.0226,Y=0.2578,Z=-0.0842,W=0.9623),Translation=(X=0,Y=0,Z=0),Scale3D=(X=1,Y=1,Z=1))),(Item=(Type=Bone,Name=\"dyn_lapel_r_01\"),Transform=(Rotation=(X=0,Y=0,Z=0,W=1),Translation=(X=0,Y=0,Z=0),Sc ...(truncated)"
    WeightMinimum=0
    Weight=RigVMVariableNode() {variable=R Arm Out Up}.Value
   )
   n3 = RigVMRerouteNode(Value=())
   n4 = &RigUnit_GetTransform(Item=(Type=Bone,Name="dyn_coat_fr_l_04"), bInitial=False)
   n5 = /Script/RigVM.RigVMFunction_MathTransformClampSpatially(
    Axis=Y
    DebugColor=(R=1,G=0,B=0,A=1)
    DebugThickness=1
    Maximum=200
    Minimum=20
    Type=Plane
    bDrawDebug=true
    Space=&RigUnit_GetTransform(Item=(Type=Control,Name="thigh_l_ctrl"), bInitial=False).Transform
    Value=@n4.Transform
   ) {
    is=struct RigVMUnitNode
    template="ClampSpatially::Execute(in Value,in Axis,in Type,in Minimum,in Maximum,in Space,in bDrawDebug,in DebugColor,in DebugThickness,out Result)"
   }
   n6 = &RigUnit_GetTransform(Item=(Type=Bone,Name="dyn_coat_fr_l_01"), bInitial=False)
   &RigUnit_OffsetTransformForItem(
    Item=(Type=Bone,Name="dyn_coat_fr_l_01")
    OffsetTransform=(Rotation=(X=0,Y=0,Z=0,W=1),Translation=(X=0,Y=0,Z=0),Scale3D=(X=1,Y=1,Z=1))
    Weight=1
    bPropagateToChildren=True
    OffsetTransform.Rotation=/Script/RigVM.RigVMFunction_MathQuaternionFromTwoVectors(
     B=&RigVMFunction_MathTransformMakeRelative(Global=@n5.Result, Parent=@n6.Transform).Local.Translation
     A=&RigVMFunction_MathTransformMakeRelative(Global=@n4.Transform, Parent=@n6.Transform).Local.Translation
    ) {is=struct RigVMUnitNode, template="FromTwoVectors::Execute(in A,in B,out Result)"}.Result
    ExecutePin=@n3.Value
   )
   RigUnit_BeginExecution = &RigUnit_BeginExecution()
   RigVMFunction_Sequence = RigVMAggregateNode(ExecuteContext=@RigUnit_BeginExecution.ExecutePin)
   n7 = &RigUnit_SphericalPoseReader(
    ActiveRegionScaleFactors=(PositiveWidth=3.5,NegativeWidth=0.5,PositiveHeight=3,NegativeHeight=3)
    ActiveRegionSize=0.1
    Debug=(bDrawDebug=false,DebugScale=25,DebugSegments=20,DebugThickness=0.25)
    DriverAxis=(X=1,Y=0,Z=0)
    DriverItem=(Type=Bone,Name="thigh_r")
    FalloffRegionScaleFactors=(PositiveWidth=1,NegativeWidth=1,PositiveHeight=1,NegativeHeight=1)
    FalloffSize=0.45
    FlipHeightScaling=false
    FlipWidthScaling=true
    OptionalParentItem=(Type=Bone)
    RotationOffset=(X=-90,Y=0,Z=0)
    ExecutePin=@RigVMFunction_Sequence.A
   )
   > RigVMVariableNode(Value=@n7.OutputParam) {variable=R Leg Fwd}
   > n8 = &RigUnit_SphericalPoseReader(
    ActiveRegionScaleFactors=(PositiveWidth=1.3,NegativeWidth=1.3,PositiveHeight=1.5,NegativeHeight=1.3)
    ActiveRegionSize=0.1
    Debug=(bDrawDebug=false,bDraw2D=false,bDrawLocalAxes=true,DebugScale=25,DebugSegments=20,DebugThickness=0.25)
    DriverAxis=(X=1,Y=0,Z=0)
    DriverItem=(Type=Bone,Name="thigh_r")
    FalloffRegionScaleFactors=(PositiveWidth=1,NegativeWidth=1,PositiveHeight=0.7,NegativeHeight=0.7)
    FalloffSize=0.4
    FlipHeightScaling=false
    FlipWidthScaling=true
    OptionalParentItem=(Type=Bone,Name="pelvis")
    RotationOffset=(X=86.3423,Y=0,Z=-47.0633)
   )
   > RigVMVariableNode(Value=@n8.OutputParam) {variable=R Leg Back}
   > n9 = &RigUnit_SphericalPoseReader(
    ActiveRegionScaleFactors=(PositiveWidth=1.5,NegativeWidth=1.5,PositiveHeight=1.5,NegativeHeight=1.5)
    ActiveRegionSize=0.1
    Debug=(bDrawDebug=false,bDraw2D=false,bDrawLocalAxes=true,DebugScale=25,DebugSegments=20,DebugThickness=0.25)
    DriverAxis=(X=1,Y=0,Z=0)
    DriverItem=(Type=Bone,Name="thigh_r")
    FalloffRegionScaleFactors=(PositiveWidth=2,NegativeWidth=1,PositiveHeight=1,NegativeHeight=2)
    FalloffSize=0.33
    FlipHeightScaling=false
    FlipWidthScaling=true
    OptionalParentItem=(Type=Bone,Name="pelvis")
    RotationOffset=(X=-24.0347,Y=-44.1547,Z=-0.9418)
   )
   > RigVMVariableNode(Value=@n9.OutputParam) {variable=R Leg Right}
   > n10 = &RigUnit_SphericalPoseReader(
    ActiveRegionScaleFactors=(PositiveWidth=1,NegativeWidth=2,PositiveHeight=1,NegativeHeight=2)
    ActiveRegionSize=0.1
    Debug=(bDrawDebug=false,bDraw2D=false,bDrawLocalAxes=true,DebugScale=25,DebugSegments=20,DebugThickness=0.25)
    DriverAxis=(X=-1,Y=0,Z=0)
    DriverItem=(Type=Bone,Name="thigh_l")
    FalloffRegionScaleFactors=(PositiveWidth=0.5,NegativeWidth=2,PositiveHeight=0.5,NegativeHeight=2)
    FalloffSize=0.4
    FlipHeightScaling=false
    FlipWidthScaling=true
    OptionalParentItem=(Type=Bone,Name="pelvis")
    RotationOffset=(X=67.0704,Y=135.4766,Z=-50.9507)
   )
   > RigVMVariableNode(Value=@n10.OutputParam) {variable=L Leg Fwd}
   > n11 = &RigUnit_SphericalPoseReader(
    ActiveRegionScaleFactors=(PositiveWidth=1,NegativeWidth=1.5,PositiveHeight=1,NegativeHeight=1)
    ActiveRegionSize=0.1
    Debug=(bDrawDebug=false,bDraw2D=false,bDrawLocalAxes=true,DebugScale=25,DebugSegments=20,DebugThickness=0.25)
    DriverAxis=(X=-1,Y=0,Z=0)
    DriverItem=(Type=Bone,Name="thigh_l")
    FalloffRegionScaleFactors=(PositiveWidth=1.5,NegativeWidth=1.5,PositiveHeight=1.5,NegativeHeight=1.5)
    FalloffSize=0.3
    FlipHeightScaling=false
    FlipWidthScaling=true
    OptionalParentItem=(Type=Bone,Name="pelvis")
    RotationOffset=(X=149.5421,Y=-39.4425,Z=3.4675)
   )
   > RigVMVariableNode(Value=@n11.OutputParam) {variable=L Leg Left}
   > n12 = &RigUnit_SphericalPoseReader(
    ActiveRegionScaleFactors=(PositiveWidth=1.5,NegativeWidth=1.5,PositiveHeight=1.5,NegativeHeight=1.5)
    ActiveRegionSize=0.1
    Debug=(bDrawDebug=false,bDraw2D=false,bDrawLocalAxes=true,DebugScale=25,DebugSegments=20,DebugThickness=0.25)
    DriverAxis=(X=-1,Y=0,Z=0)
    DriverItem=(Type=Bone,Name="thigh_l")
    FalloffRegionScaleFactors=(PositiveWidth=1.3,NegativeWidth=1.3,PositiveHeight=0.7,NegativeHeight=0.7)
    FalloffSize=0.4
    FlipHeightScaling=false
    FlipWidthScaling=true
    OptionalParentItem=(Type=Bone,Name="pelvis")
    RotationOffset=(X=-95.4504,Y=0,Z=-45.1954)
   )
   > RigVMVariableNode(Value=@n12.OutputParam) {variable=L Leg Back}
   > n13 = &RigUnit_SphericalPoseReader(
    ActiveRegionScaleFactors=(PositiveWidth=2,NegativeWidth=1,PositiveHeight=1,NegativeHeight=1)
    ActiveRegionSize=0.1
    Debug=(bDrawDebug=false,bDraw2D=false,bDrawLocalAxes=true,DebugScale=25,DebugSegments=20,DebugThickness=0.25)
    DriverAxis=(X=1,Y=0,Z=0)
    DriverItem=(Type=Bone,Name="thigh_r")
    FalloffRegionScaleFactors=(PositiveWidth=1,NegativeWidth=1,PositiveHeight=1,NegativeHeight=1)
    FalloffSize=0.25
    FlipHeightScaling=false
    FlipWidthScaling=true
    OptionalParentItem=(Type=Bone,Name="spine_02")
    RotationOffset=(X=-87.9328,Y=0,Z=58.3289)
   )
   > RigVMVariableNode(Value=@n13.OutputParam) {variable=R Thigh Run}
   > n14 = &RigUnit_SphericalPoseReader(
    ActiveRegionScaleFactors=(PositiveWidth=1,NegativeWidth=1,PositiveHeight=1,NegativeHeight=1)
    ActiveRegionSize=0.1
    Debug=(bDrawDebug=false,bDraw2D=false,bDrawLocalAxes=true,DebugScale=25,DebugSegments=20,DebugThickness=0.25)
    DriverAxis=(X=1,Y=0,Z=0)
    DriverItem=(Type=Bone,Name="calf_r")
    FalloffRegionScaleFactors=(PositiveWidth=0.3,NegativeWidth=1.5,PositiveHeight=1,NegativeHeight=1)
    FalloffSize=0.25
    FlipHeightScaling=false
    FlipWidthScaling=true
    OptionalParentItem=(Type=Bone,Name="spine_02")
    RotationOffset=(X=83.6243,Y=0,Z=-14.2191)
   )
   > RigVMVariableNode(Value=@n14.OutputParam) {variable=R Calf Run}
   > n15 = &RigUnit_SphericalPoseReader(
    ActiveRegionScaleFactors=(PositiveWidth=2,NegativeWidth=2,PositiveHeight=2,NegativeHeight=2)
    ActiveRegionSize=0.1
    Debug=(bDrawDebug=false,bDraw2D=false,bDrawLocalAxes=true,DebugScale=25,DebugSegments=20,DebugThickness=0.25)
    DriverAxis=(X=-1,Y=0,Z=0)
    DriverItem=(Type=Bone,Name="thigh_l")
    FalloffRegionScaleFactors=(PositiveWidth=1,NegativeWidth=1,PositiveHeight=1,NegativeHeight=1)
    FalloffSize=0.25
    FlipHeightScaling=false
    FlipWidthScaling=true
    OptionalParentItem=(Type=Bone,Name="spine_02")
    RotationOffset=(X=-97.7291,Y=0,Z=-112.0332)
   )
   > RigVMVariableNode(Value=@n15.OutputParam) {variable=L Thigh Run}
   > n16 = &RigUnit_SphericalPoseReader(
    ActiveRegionScaleFactors=(PositiveWidth=2,NegativeWidth=2,PositiveHeight=2,NegativeHeight=2)
    ActiveRegionSize=0.1
    Debug=(bDrawDebug=false,bDraw2D=false,bDrawLocalAxes=true,DebugScale=25,DebugSegments=20,DebugThickness=0.25)
    DriverAxis=(X=-1,Y=0,Z=0)
    DriverItem=(Type=Bone,Name="calf_l")
    FalloffRegionScaleFactors=(PositiveWidth=0.3,NegativeWidth=1.5,PositiveHeight=1,NegativeHeight=1)
    FalloffSize=0.25
    FlipHeightScaling=false
    FlipWidthScaling=true
    OptionalParentItem=(Type=Bone,Name="spine_02")
    RotationOffset=(X=83.6243,Y=0,Z=159.8158)
   )
   > RigVMVariableNode(Value=@n16.OutputParam) {variable=L Calf Run}
   > n17 = &RigUnit_SphericalPoseReader(
    ActiveRegionScaleFactors=(PositiveWidth=2,NegativeWidth=2.5,PositiveHeight=2,NegativeHeight=2.3)
    ActiveRegionSize=0.11
    Debug=(bDrawDebug=false,bDraw2D=false,bDrawLocalAxes=true,DebugScale=25,DebugSegments=20,DebugThickness=0.25)
    DriverAxis=(X=-1,Y=0,Z=0)
    DriverItem=(Type=Bone,Name="upperarm_r")
    FalloffRegionScaleFactors=(PositiveWidth=0.3,NegativeWidth=1.5,PositiveHeight=2.2,NegativeHeight=1)
    FalloffSize=0.35
    FlipHeightScaling=false
    FlipWidthScaling=true
    OptionalParentItem=(Type=Bone,Name="spine_02")
    RotationOffset=(X=-140.847,Y=44.4196,Z=9.9848)
   )
   > RigVMVariableNode(Value=@n17.OutputParam) {variable=R Arm Up_Full}
   > n18 = &RigUnit_SphericalPoseReader(
    ActiveRegionScaleFactors=(PositiveWidth=1.5,NegativeWidth=1.5,PositiveHeight=1.5,NegativeHeight=1.2)
    ActiveRegionSize=0.11
    Debug=(bDrawDebug=false,bDraw2D=false,bDrawLocalAxes=true,DebugScale=25,DebugSegments=20,DebugThickness=0.25)
    DriverAxis=(X=-1,Y=0,Z=0)
    DriverItem=(Type=Bone,Name="upperarm_r")
    FalloffRegionScaleFactors=(PositiveWidth=1,NegativeWidth=1,PositiveHeight=0.5,NegativeHeight=1)
    FalloffSize=0.35
    FlipHeightScaling=false
    FlipWidthScaling=true
    OptionalParentItem=(Type=Bone,Name="spine_02")
    RotationOffset=(X=-86.8053,Y=32.9054,Z=-5.2038)
   )
   > RigVMVariableNode(Value=@n18.OutputParam) {variable=R Arm Up}
   > n19 = &RigUnit_SphericalPoseReader(
    ActiveRegionScaleFactors=(PositiveWidth=2,NegativeWidth=2,PositiveHeight=2,NegativeHeight=2)
    ActiveRegionSize=0.11
    Debug=(bDrawDebug=false,bDraw2D=false,bDrawLocalAxes=true,DebugScale=25,DebugSegments=20,DebugThickness=0.25)
    DriverAxis=(X=-1,Y=0,Z=0)
    DriverItem=(Type=Bone,Name="upperarm_r")
    FalloffRegionScaleFactors=(PositiveWidth=1,NegativeWidth=1,PositiveHeight=1,NegativeHeight=1)
    FalloffSize=0.35
    FlipHeightScaling=false
    FlipWidthScaling=true
    OptionalParentItem=(Type=Bone,Name="spine_02")
    RotationOffset=(X=-162.5802,Y=3.4496,Z=-76.7731)
   )
   > RigVMVariableNode(Value=@n19.OutputParam) {variable=R Arm Out Up}
   > n20 = &RigUnit_SphericalPoseReader(
    ActiveRegionScaleFactors=(PositiveWidth=3,NegativeWidth=1.5,PositiveHeight=3,NegativeHeight=1.2)
    ActiveRegionSize=0.11
    Debug=(bDrawDebug=false,bDraw2D=false,bDrawLocalAxes=true,DebugScale=25,DebugSegments=20,DebugThickness=0.25)
    DriverAxis=(X=-1,Y=0,Z=0)
    DriverItem=(Type=Bone,Name="upperarm_r")
    FalloffRegionScaleFactors=(PositiveWidth=3,NegativeWidth=1,PositiveHeight=3,NegativeHeight=1)
    FalloffSize=0.35
    FlipHeightScaling=false
    FlipWidthScaling=true
    OptionalParentItem=(Type=Bone,Name="spine_02")
    RotationOffset=(X=2.2339,Y=84.0009,Z=-12.5702)
   )
   > RigVMVariableNode(Value=@n20.OutputParam) {variable=R Arm Back}
   > n21 = &RigUnit_SphericalPoseReader(
    ActiveRegionScaleFactors=(PositiveWidth=2,NegativeWidth=2,PositiveHeight=2,NegativeHeight=2)
    ActiveRegionSize=0.11
    Debug=(bDrawDebug=false,bDraw2D=false,bDrawLocalAxes=true,DebugScale=25,DebugSegments=20,DebugThickness=0.25)
    DriverAxis=(X=1,Y=0,Z=0)
    DriverItem=(Type=Bone,Name="upperarm_l")
    FalloffRegionScaleFactors=(PositiveWidth=1,NegativeWidth=1,PositiveHeight=1,NegativeHeight=1)
    FalloffSize=0.35
    FlipHeightScaling=false
    FlipWidthScaling=true
    OptionalParentItem=(Type=Bone,Name="spine_02")
    RotationOffset=(X=6.8499,Y=-16.4695,Z=-3.0512)
   )
   > RigVMVariableNode(Value=@n21.OutputParam) {variable=L Arm Out Up}
   > n22 = &RigUnit_SphericalPoseReader(
    ActiveRegionScaleFactors=(PositiveWidth=3,NegativeWidth=2,PositiveHeight=3,NegativeHeight=3)
    ActiveRegionSize=0.12
    Debug=(bDrawDebug=false,DebugScale=25,DebugSegments=20,DebugThickness=0.25)
    DriverAxis=(X=-1,Y=0,Z=0)
    DriverItem=(Type=Bone,Name="thigh_l")
    FalloffRegionScaleFactors=(PositiveWidth=1,NegativeWidth=1,PositiveHeight=1,NegativeHeight=1)
    FalloffSize=0.45
    OptionalParentItem=(Type=Bone,Name="pelvis")
    RotationOffset=(X=180,Y=0,Z=0)
   )
   > RigVMVariableNode(Value=@n22.OutputParam) {variable=L Leg Out}
   > n23 = &RigUnit_SphericalPoseReader(
    ActiveRegionScaleFactors=(PositiveWidth=1,NegativeWidth=1,PositiveHeight=1,NegativeHeight=1)
    ActiveRegionSize=0.1
    Debug=(bDrawDebug=true,bDraw2D=false,bDrawLocalAxes=true,DebugScale=25,DebugSegments=20,DebugThickness=0.25)
    DriverAxis=(X=-1,Y=0,Z=0)
    DriverItem=(Type=Bone,Name="thigh_l")
    FalloffRegionScaleFactors=(PositiveWidth=0.5,NegativeWidth=0.5,PositiveHeight=0.5,NegativeHeight=0.5)
    FalloffSize=0.4
    FlipHeightScaling=false
    FlipWidthScaling=true
    OptionalParentItem=(Type=Bone,Name="pelvis")
    RotationOffset=(X=67.0704,Y=135.4766,Z=-80.5222)
   )
   > RigVMVariableNode(Value=@n23.OutputParam) {variable=L Leg Fwd_EXTREME}
   &RigUnit_ModifyTransforms(
    ItemToModify="((Item=(Type=Bone,Name=\"dyn_coat_fr_r_base\"),Transform=(Rotation=(X=-0.3976,Y=-0.1475,Z=0.2165,W=0.8794),Translation=(X=-1.146,Y=8.044,Z=1.5424),Scale3D=(X=1,Y=1,Z=1))),(Item=(Type=Bone,Name=\"dyn_coat_side_r_base\"),Transform=(Rotation=(X=-0.0013,Y=-0.1659, ...(truncated)"
    ExecutePin=@RigVMFunction_Sequence.B
    Weight=RigVMVariableNode() {variable=R Leg Right}.Value
   )
   > &RigUnit_ModifyTransforms(
    ItemToModify="((Item=(Type=Bone,Name=\"dyn_coat_fr_r_base\"),Transform=(Rotation=(X=0,Y=0,Z=0.2164,W=0.9763),Translation=(X=-9.0082,Y=9.3096,Z=6),Scale3D=(X=1,Y=1,Z=1))),(Item=(Type=Bone,Name=\"dyn_coat_side_r_base\"),Transform=(Rotation=(X=0,Y=0,Z=0,W=1),Translation=(X=-5, ...(truncated)"
    Weight=RigVMVariableNode() {variable=R Leg Fwd}.Value
   )
   > &RigUnit_ModifyTransforms(
    ItemToModify="((Item=(Type=Bone,Name=\"dyn_coat_side_r_base\"),Transform=(Rotation=(X=0,Y=-0.1305,Z=0,W=0.9914),Translation=(X=0,Y=0,Z=5),Scale3D=(X=1,Y=1,Z=1))),(Item=(Type=Bone,Name=\"dyn_coat_bck_r_base\"),Transform=(Rotation=(X=0,Y=-0.342,Z=0,W=0.9397),Translation=(X=0, ...(truncated)"
    Weight=RigVMVariableNode() {variable=R Leg Back}.Value
   )
   > &RigUnit_ModifyTransforms(
    ItemToModify="((Item=(Type=Bone,Name=\"dyn_coat_fr_l_base\"),Transform=(Rotation=(X=0.1297,Y=-0.0508,Z=-0.0063,W=0.9902),Translation=(X=-19.7852,Y=-10.6686,Z=-1.1676),Scale3D=(X=1,Y=1,Z=1))),(Item=(Type=Bone,Name=\"dyn_coat_fr_l_02\"),Transform=(Rotation=(X=0,Y=0,Z=0,W=1),T ...(truncated)"
    Weight=RigVMVariableNode() {variable=L Leg Fwd}.Value
   )
   > &RigUnit_ModifyTransforms(
    ItemToModify="((Item=(Type=Bone,Name=\"dyn_coat_fr_l_base\"),Transform=(Rotation=(X=0.1573,Y=-0.2039,Z=-0.1957,W=0.9462),Translation=(X=-1.3112,Y=0.5263,Z=3.1937),Scale3D=(X=1,Y=1,Z=1))),(Item=(Type=Bone,Name=\"dyn_coat_fr_l_02\"),Transform=(Rotation=(X=0.0038,Y=0.0871,Z=0. ...(truncated)"
    Weight=RigVMVariableNode() {variable=L Leg Left}.Value
   )
   > &RigUnit_ModifyTransforms(
    ItemToModify="((Item=(Type=Bone,Name=\"dyn_coat_bck_l_base\"),Transform=(Rotation=(X=-0.0868,Y=-0.2578,Z=-0.0076,W=0.9623),Translation=(X=0,Y=0,Z=0),Scale3D=(X=1,Y=1,Z=1))),(Item=(Type=Bone,Name=\"dyn_coat_side_l_base\"),Transform=(Rotation=(X=-0.0076,Y=-0.0867,Z=-0.0433,W= ...(truncated)"
    Weight=RigVMVariableNode() {variable=L Leg Back}.Value
   )
   > &RigUnit_ModifyTransforms(
    ItemToModify="((Item=(Type=Bone,Name=\"dyn_sleeve_r_fr\"),Transform=(Rotation=(X=0,Y=0,Z=0,W=1),Translation=(X=0,Y=0,Z=-1.9187),Scale3D=(X=1,Y=1,Z=1))),(Item=(Type=Bone,Name=\"dyn_lapel_r_02\"),Transform=(Rotation=(X=0,Y=0,Z=0,W=1),Translation=(X=0,Y=0,Z=0),Scale3D=(X=1,Y=1 ...(truncated)"
    WeightMinimum=0
    Weight=RigVMVariableNode() {variable=R Arm Up_Full}.Value
   )
   > &RigUnit_ModifyTransforms(
    ItemToModify="((Item=(Type=Bone,Name=\"dyn_lapel_r_01\"),Transform=(Rotation=(X=-0.083,Y=-0.0905,Z=0.0357,W=0.9918),Translation=(X=0,Y=0.2,Z=0.5),Scale3D=(X=1,Y=1,Z=1))),(Item=(Type=Bone,Name=\"dyn_sleeve_r_fr\"),Transform=(Rotation=(X=0,Y=0.0872,Z=0,W=0.9962),Translation=( ...(truncated)"
    WeightMinimum=0
    Weight=RigVMVariableNode() {variable=R Arm Back}.Value
   )
   > n24 = &RigUnit_ModifyTransforms(
    ItemToModify="((Item=(Type=Bone,Name=\"dyn_coat_fr_l_base\"),Transform=(Rotation=(X=0.3384,Y=-0.3011,Z=0.0628,W=0.8893),Translation=(X=-10.2593,Y=-7.1611,Z=-11.4298),Scale3D=(X=1,Y=1,Z=1))),(Item=(Type=Bone,Name=\"dyn_coat_fr_l_02\"),Transform=(Rotation=(X=0,Y=0,Z=0,W=1),Tr ...(truncated)"
    Weight=RigVMVariableNode() {variable=L Leg Fwd_EXTREME}.Value
   )
   n25 = &RigVMFunction_ControlFlowBranch(ExecuteContext=@n24.ExecutePin, Condition=RigVMVariableNode() {variable=Is Playing Emote}.Value)
   &RigUnit_ModifyTransforms(
    ItemToModify="((Item=(Type=Bone,Name=\"dyn_coat_fr_l_base\"),Transform=(Rotation=(X=-0.017,Y=-0.1294,Z=-0.1294,W=0.983),Translation=(X=0,Y=0,Z=3),Scale3D=(X=1,Y=1,Z=1))),(Item=(Type=Bone,Name=\"dyn_coat_fr_l_01\"),Transform=(Rotation=(X=0,Y=0,Z=0,W=1),Translation=(X=0,Y=0,Z ...(truncated)"
    ExecutePin=@n25.True
    Weight=RigVMVariableNode() {variable=L Leg Out}.Value
   )
   n26 = &RigUnit_GetTransform(Item=(Type=Bone,Name="dyn_lapel_r_01"), bInitial=False)
   n27 = &RigUnit_GetTransform(Item=(Type=Bone,Name="dyn_lapel_r_02"), bInitial=False)
   ClampSpatially = &RigVMFunction_MathTransformClampSpatially(
    Axis=Y
    DebugColor=(R=1,G=0,B=0,A=1)
    DebugThickness=1
    Maximum=-200
    Minimum=5
    Type=Plane
    bDrawDebug=false
    Value=@n27.Transform
    Space=&RigUnit_GetTransform(Item=(Type=Control,Name="upperarm_r_ctrl_2"), bInitial=False).Transform
   ) {template="ClampSpatially::Execute(in Value,in Axis,in Type,in Minimum,in Maximum,in Space,in bDrawDebug,in DebugColor,in DebugThickness,out Result)"}
   &RigUnit_OffsetTransformForItem(
    Item=(Type=Bone,Name="dyn_lapel_r_01")
    OffsetTransform=(Rotation=(X=0,Y=0,Z=0,W=1),Translation=(X=0,Y=0,Z=0),Scale3D=(X=1,Y=1,Z=1))
    Weight=1
    bPropagateToChildren=True
    OffsetTransform.Rotation=&RigVMFunction_MathQuaternionFromTwoVectors(
     A=&RigVMFunction_MathTransformMakeRelative(Parent=@n26.Transform, Global=@n27.Transform).Local.Translation
     B=&RigVMFunction_MathTransformMakeRelative(Global=@ClampSpatially.Result, Parent=@n26.Transform).Local.Translation
    ) {template="FromTwoVectors::Execute(in A,in B,out Result)"}.Result
    ExecutePin=@RigVMFunction_Sequence.C
   )
   n28 = &RigUnit_GetTransform(Item=(Type=Bone,Name="dyn_lapel_l_02"), bInitial=False)
   n29 = &RigVMFunction_MathTransformClampSpatially(
    Axis=Y
    DebugColor=(R=1,G=0,B=0,A=1)
    DebugThickness=1
    Maximum=-200
    Minimum=6
    Type=Plane
    bDrawDebug=false
    Space=&RigUnit_GetTransform(Item=(Type=Control,Name="upperarm_l_ctrl"), bInitial=False).Transform
    Value=@n28.Transform
   ) {template="ClampSpatially::Execute(in Value,in Axis,in Type,in Minimum,in Maximum,in Space,in bDrawDebug,in DebugColor,in DebugThickness,out Result)"}
   n30 = &RigUnit_GetTransform(Item=(Type=Bone,Name="dyn_lapel_l_01"), bInitial=False)
   > &RigUnit_OffsetTransformForItem(
    Item=(Type=Bone,Name="dyn_lapel_l_01")
    OffsetTransform=(Rotation=(X=0,Y=0,Z=0,W=1),Translation=(X=0,Y=0,Z=0),Scale3D=(X=1,Y=1,Z=1))
    Weight=1
    bPropagateToChildren=True
    OffsetTransform.Rotation=&RigVMFunction_MathQuaternionFromTwoVectors(
     B=&RigVMFunction_MathTransformMakeRelative(Global=@n29.Result, Parent=@n30.Transform).Local.Translation
     A=&RigVMFunction_MathTransformMakeRelative(Global=@n28.Transform, Parent=@n30.Transform).Local.Translation
    ) {template="FromTwoVectors::Execute(in A,in B,out Result)"}.Result
   )
   > &RigUnit_SetTransform(
    Item=(Type=Bone,Name="dyn_sleeve_r_fr")
    Weight=1
    bInitial=False
    Value=&RigVMFunction_MathTransformClampSpatially(
     Axis=Z
     DebugColor=(R=1,G=0,B=0,A=1)
     DebugThickness=1
     Maximum=200
     Minimum=4
     Type=Plane
     bDrawDebug=false
     Space=&RigUnit_GetTransform(Item=(Type=Control,Name="spine_05_ctrl"), bInitial=False).Transform
     Value=&RigUnit_GetTransform(Item=(Type=Bone,Name="dyn_sleeve_r_fr"), bInitial=False).Transform
    ) {template="ClampSpatially::Execute(in Value,in Axis,in Type,in Minimum,in Maximum,in Space,in bDrawDebug,in DebugColor,in DebugThickness,out Result)"}.Result
   )
   RigVMFunction_MathQuaternionToEuler = &RigVMFunction_MathQuaternionToEuler(
    RotationOrder=ZYX
    Value=&RigUnit_GetRelativeTransformForItem(Child=(Type=Bone,Name="dyn_lapel_r_01"), Parent=(Type=Bone,Name="spine_05"), bChildInitial=False, bParentInitial=False).RelativeTransform.Rotation
   )
   Remap = &RigVMFunction_MathDoubleRemap(SourceMaximum=-7, SourceMinimum=-30, TargetMaximum=1, TargetMinimum=0, Value=@RigVMFunction_MathQuaternionToEuler.Result.Z)
   > RigVMVariableNode(Value=@Remap.Result) {variable=R Lapel Moving}
   AlphaInterp = &RigVMFunction_AlphaInterp(
    InterpSpeedDecreasing=10
    InterpSpeedIncreasing=10
    Scale=1
    bClampResult=False
    bInterpResult=False
    Value=RigVMVariableNode() {variable=R Lapel Moving}.Value
   )
   > &RigUnit_ModifyTransforms(
    ItemToModify="((Item=(Type=Bone,Name=\"dyn_lapel_r_01\"),Transform=(Rotation=(X=0,Y=0,Z=-0.1305,W=0.9914),Translation=(X=0,Y=0,Z=0),Scale3D=(X=1,Y=1,Z=1))),(Item=(Type=Bone,Name=\"dyn_lapel_r_02\"),Transform=(Rotation=(X=0,Y=0,Z=-0.342,W=0.9397),Translation=(X=0,Y=0,Z=0),Sc ...(truncated)"
    Weight=@AlphaInterp.Result
   )
   &RigVMFunction_AlphaInterp(InterpSpeedDecreasing=10, InterpSpeedIncreasing=10, Scale=1, Value=0, bClampResult=False, bInterpResult=False)
   &RigUnit_ModifyTransforms(
    ExecutePin=()
    ItemToModify="((Item=(Type=Bone,Name=\"dyn_coat_fr_r_base\"),Transform=(Rotation=(X=-0.3976,Y=-0.1475,Z=0.2165,W=0.8794),Translation=(X=-4,Y=1,Z=3),Scale3D=(X=1,Y=1,Z=1))),(Item=(Type=Bone,Name=\"dyn_coat_side_r_base\"),Transform=(Rotation=(X=-0.0013,Y=-0.1659,Z=0.0916,W=0. ...(truncated)"
    Weight=1
   )
   n31 = &RigUnit_GetTransform(Item=(Type=Bone,Name="dyn_coat_bck_l_04"), bInitial=False)
   n32 = &RigVMFunction_MathTransformClampSpatially(
    Axis=Y
    DebugColor=(R=1,G=0,B=0,A=1)
    DebugThickness=1
    Maximum=200
    Minimum=5
    Type=Plane
    bDrawDebug=true
    Space=&RigUnit_GetTransform(Item=(Type=Control,Name="thigh_l_ctrl"), bInitial=False).Transform
    Value=@n31.Transform
   ) {template="ClampSpatially::Execute(in Value,in Axis,in Type,in Minimum,in Maximum,in Space,in bDrawDebug,in DebugColor,in DebugThickness,out Result)"}
   n33 = &RigUnit_GetTransform(Item=(Type=Bone,Name="dyn_coat_fr_l_01"), bInitial=False)
   &RigUnit_OffsetTransformForItem(
    ExecutePin=()
    Item=(Type=Bone,Name="dyn_coat_fr_l_01")
    OffsetTransform=(Rotation=(X=0,Y=0,Z=0,W=1),Translation=(X=0,Y=0,Z=0),Scale3D=(X=1,Y=1,Z=1))
    Weight=1
    bPropagateToChildren=True
    OffsetTransform.Rotation=&RigVMFunction_MathQuaternionFromTwoVectors(
     B=&RigVMFunction_MathTransformMakeRelative(Global=@n32.Result, Parent=@n33.Transform).Local.Translation
     A=&RigVMFunction_MathTransformMakeRelative(Global=@n31.Transform, Parent=@n33.Transform).Local.Translation
    ) {template="FromTwoVectors::Execute(in A,in B,out Result)"}.Result
   )
   &RigUnit_ParentConstraint(
    AdvancedSettings=(InterpolationType=Average,RotationOrderForFilter=XZY)
    Child=(Type=Bone,Name="dyn_coat_fr_l_01")
    ExecutePin=()
    Parents=((Item=(Type=Bone,Name="thigh_l"),Weight=1))
    Weight=1
   )
   &RigUnit_ParentConstraint(
    AdvancedSettings=(InterpolationType=Average,RotationOrderForFilter=XZY)
    Child=(Type=Bone,Name="dyn_coat_fr_l_01")
    ExecutePin=()
    Parents=((Item=(Type=Bone,Name="thigh_l"),Weight=1))
    Weight=1
   )
   RigVMFunction_ControlFlowBranch = &RigVMFunction_ControlFlowBranch(ExecuteContext=(), Condition=RigVMVariableNode() {variable=Is Targeting}.Value)
   &RigVMFunction_ControlFlowBranch(ExecuteContext=(), Condition=RigVMVariableNode() {variable=Is Targeting}.Value)
   &RigUnit_SphericalPoseReader(
    ActiveRegionScaleFactors=(PositiveWidth=1.5,NegativeWidth=1.5,PositiveHeight=1.5,NegativeHeight=1.5)
    ActiveRegionSize=0.1
    Debug=(bDrawDebug=false,bDraw2D=false,bDrawLocalAxes=true,DebugScale=25,DebugSegments=20,DebugThickness=0.25)
    DriverAxis=(X=1,Y=0,Z=0)
    DriverItem=(Type=Bone,Name="thigh_r")
    ExecutePin=()
    FalloffRegionScaleFactors=(PositiveWidth=0.5,NegativeWidth=2,PositiveHeight=0.5,NegativeHeight=2)
    FalloffSize=0.4
    FlipHeightScaling=false
    FlipWidthScaling=true
    OptionalParentItem=(Type=Bone,Name="pelvis")
    RotationOffset=(X=-57.4038,Y=35.1015,Z=-21.5201)
   )
   &RigUnit_SphericalPoseReader(
    ActiveRegionScaleFactors=(PositiveWidth=1,NegativeWidth=1,PositiveHeight=1,NegativeHeight=1)
    ActiveRegionSize=0.15
    Debug=(bDrawDebug=false,bDraw2D=false,bDrawLocalAxes=true,DebugScale=25,DebugSegments=20,DebugThickness=0.25)
    DriverAxis=(X=-14.4715,Y=-0.7267,Z=-1.9608)
    DriverItem=(Type=Bone,Name="thigh_l")
    ExecutePin=()
    FalloffRegionScaleFactors=(PositiveWidth=1,NegativeWidth=1,PositiveHeight=1,NegativeHeight=1)
    FalloffSize=0.2
    FlipHeightScaling=false
    FlipWidthScaling=true
    OptionalParentItem=(Type=Bone,Name="spine_02")
    RotationOffset=(X=164.6035,Y=9.9672,Z=37.4503)
   )
   &RigUnit_SphericalPoseReader(
    ActiveRegionScaleFactors=(PositiveWidth=1.4,NegativeWidth=1.4,PositiveHeight=1.4,NegativeHeight=1.4)
    ActiveRegionSize=0.15
    Debug=(bDrawDebug=false,bDraw2D=false,bDrawLocalAxes=true,DebugScale=25,DebugSegments=20,DebugThickness=0.25)
    DriverAxis=(X=1,Y=0,Z=0)
    DriverItem=(Type=Bone,Name="thigh_r")
    ExecutePin=()
    FalloffRegionScaleFactors=(PositiveWidth=1,NegativeWidth=1,PositiveHeight=1,NegativeHeight=1)
    FalloffSize=0.3
    FlipHeightScaling=false
    FlipWidthScaling=true
    OptionalParentItem=(Type=Bone,Name="spine_02")
    RotationOffset=(X=-32.6002,Y=-54.6877,Z=0)
   )
  ]}
]
hierarchy [
 root {loc=<0 0 0>, rot=<0 0 0 1>}
 pelvis {parent=root, loc=<0 -0.5301 98.0571>, rot=<0.0127 0.707 -0.0127 -0.707>}
 spine_01 {parent=pelvis, loc=<6.1208 -0.9328 0>, rot=<0 0 0.034 -0.9994>}
 spine_02 {parent=spine_01, loc=<6.5996 -0.2095 0>, rot=<0 0 0.0238 -0.9997>}
 spine_03 {parent=spine_02, loc=<6.4914 -0.1237 0>, rot=<0 0 0.0149 -0.9999>}
 spine_04 {parent=spine_03, loc=<6.9598 0.0794 0>, rot=<0 0 -0.0132 -0.9999>}
 spine_05 {parent=spine_04, loc=<6.9987 0 0>, rot=<0 0 -0.0015 -1>}
 attach_cape {parent=spine_05, loc=<9.3216 -8.4501 0>, rot=<0.0616 -0.7044 -0.0616 -0.7044>}
 neck_01 {parent=spine_05, loc=<11.3138 0.1529 0>, rot=<0 0 -0.2003 -0.9797>}
 neck_02 {parent=neck_01, loc=<4.5627 0 0>, rot=<0 0 0.0504 -0.9987>}
 head {parent=neck_02, loc=<5.3914 0 0>, rot=<0 0 0.115 -0.9934>}
 head_ctrl {controlType=EULER_TRANSFORM, parent=head, type=Control, loc=<0 0 0>, rot=<0 0 0 1>}
 attach_backpack {parent=spine_05, loc=<0 -8.35 0>, rot=<-0.0093 -0.707 0.0093 -0.707>}
 clavicle_l {parent=spine_05, loc=<5.4831 -0.1053 -3.2946>, rot=<0.0465 -0.7289 -0.009 -0.6829>}
 upperarm_l {parent=clavicle_l, loc=<9.8658 -1.2136 -0.4105>, rot=<0.0164 -0.3245 -0.0269 -0.9454>}
 lowerarm_l {parent=upperarm_l, loc=<26.465 -0.8345 0.259>, rot=<0.0036 0.0161 -0.2298 -0.9731>}
 hand_l {parent=lowerarm_l, loc=<22.7355 -0.3203 -0.6105>, rot=<-0.6784 -0.0223 0.0735 -0.7307>}
 thumb_01_l {parent=hand_l, loc=<1.3118 -0.3191 -2.459>, rot=<0.6546 -0.4042 0.0664 -0.6354>}
 thumb_02_l {parent=thumb_01_l, loc=<5.9336 -0.0448 0.0401>, rot=<-0.0037 0.0213 0.298 -0.9543>}
 thumb_03_l {parent=thumb_02_l, loc=<2.2223 0.3682 -0.0907>, rot=<0.0118 -0.013 0.0066 -0.9998>}
 index_metacarpal_l {parent=hand_l, loc=<1.6535 0.0183 -2.96>, rot=<0.0158 -0.0124 -0.0173 -0.9996>}
 index_01_l {parent=index_metacarpal_l, loc=<8.3008 0.0357 -0.0524>, rot=<-0.0044 0.0485 0.0924 -0.9945>}
 index_02_l {parent=index_01_l, loc=<3.3587 -0.3004 0.0452>, rot=<-0.002 0.0027 0.3225 -0.9466>}
 index_03_l {parent=index_02_l, loc=<2.5578 -0.123 -0.0252>, rot=<-0.002 0.0067 0.2558 -0.9667>}
 middle_metacarpal_l {parent=hand_l, loc=<2.1856 -0.1725 -0.5984>, rot=<0.006 0.0168 -0.0329 -0.9993>}
 middle_01_l {parent=middle_metacarpal_l, loc=<7.3275 -0.2447 0.0129>, rot=<-0.0208 0.0042 0.0858 -0.9961>}
 middle_02_l {parent=middle_01_l, loc=<4.0123 0.1401 0.1175>, rot=<-0.0218 0.0258 0.3512 -0.9357>}
 middle_03_l {parent=middle_02_l, loc=<2.6077 0.0534 0.1373>, rot=<0.0082 0.0188 0.1708 -0.9851>}
 ring_metacarpal_l {parent=hand_l, loc=<2.2074 -0.3625 1.3638>, rot=<-0.0284 0.0722 -0.0458 -0.9959>}
 ring_01_l {parent=ring_metacarpal_l, loc=<7.0295 -0.6197 -0.1446>, rot=<-0.0489 -0.0252 0.1892 -0.9804>}
 ring_02_l {parent=ring_01_l, loc=<3.6174 0.4647 -0.1027>, rot=<0 0 0.2806 -0.9598>}
 ring_03_l {parent=ring_02_l, loc=<2.4152 0.114 0.0199>, rot=<0 0 0.1189 -0.9929>}
 pinky_metacarpal_l {parent=hand_l, loc=<1.7616 -0.9753 3.3814>, rot=<-0.0397 0.0814 -0.069 -0.9935>}
 pinky_01_l {parent=pinky_metacarpal_l, loc=<7.1709 -0.4831 -0.1094>, rot=<-0.0654 -0.0035 0.2242 -0.9723>}
 pinky_02_l {parent=pinky_01_l, loc=<2.8274 0.0893 -0.1043>, rot=<0.0037 0.0011 0.2854 -0.9584>}
 pinky_03_l {parent=pinky_02_l, loc=<2.2046 0.0307 -0.0078>, rot=<-0.0002 0.0098 0.1445 -0.9895>}
 lowerarm_twist_02_l {parent=lowerarm_l, loc=<10.6707 0 0>, rot=<0 0 0 1>}
 lowerarm_twist_01_l {parent=lowerarm_l, loc=<18.316 0 0>, rot=<0 0 0 1>}
 upperarm_twist_01_l {parent=upperarm_l, loc=<7.2135 -0.1134 0>, rot=<0 0 0 1>}
 dyn_sleeve_l_base {parent=upperarm_twist_01_l, loc=<0.0002 0.0015 -0.0002>, rot=<-0.0557 0.0111 0.0287 -0.998>}
 dyn_sleeve_l_fr {parent=dyn_sleeve_l_base, loc=<-1.3421 5.0465 0.1452>, rot=<0.6628 0.0582 -0.0941 -0.7406>}
 dyn_sleeve_l_top {parent=dyn_sleeve_l_base, loc=<-2.1747 -1.052 6.2755>, rot=<-0.0145 0.1437 -0.0195 -0.9893>}
 dyn_sleeve_l_bck {parent=dyn_sleeve_l_base, loc=<-0.997 -7.0206 1.1689>, rot=<-0.559 0.1348 -0.0213 -0.8178>}
 dyn_sleeve_l_bot {parent=dyn_sleeve_l_base, loc=<0.2595 -1.0106 -6.1358>, rot=<0.0313 -0.1169 0.0077 -0.9926>}
 upperarm_twist_02_l {parent=upperarm_l, loc=<19.7998 -0.1134 0>, rot=<0 0 0 1>}
 upperarm_l_ctrl {controlType=EULER_TRANSFORM, parent=upperarm_l, type=Control, loc=<0 0 0>, rot=<0 0 0 1>}
 clavicle_r {parent=spine_05, loc=<5.4831 -0.1053 3.2946>, rot=<0.7289 0.0465 -0.6829 0.009>}
 upperarm_r {parent=clavicle_r, loc=<-9.8658 1.2136 0.4105>, rot=<0.0164 -0.3245 -0.0269 -0.9454>}
 lowerarm_r {parent=upperarm_r, loc=<-26.465 0.8345 -0.259>, rot=<0.0036 0.0161 -0.2298 -0.9731>}
 hand_r {parent=lowerarm_r, loc=<-22.7355 0.3203 0.6106>, rot=<-0.6784 -0.0223 0.0735 -0.7307>}
 thumb_01_r {parent=hand_r, loc=<-1.3118 0.3191 2.459>, rot=<0.6546 -0.4042 0.0664 -0.6354>}
 thumb_02_r {parent=thumb_01_r, loc=<-5.9336 0.0448 -0.0401>, rot=<-0.0037 0.0213 0.298 -0.9543>}
 thumb_03_r {parent=thumb_02_r, loc=<-2.2223 -0.3682 0.0907>, rot=<0.0118 -0.013 0.0066 -0.9998>}
 index_metacarpal_r {parent=hand_r, loc=<-1.6535 -0.0183 2.96>, rot=<0.0158 -0.0124 -0.0173 -0.9996>}
 index_01_r {parent=index_metacarpal_r, loc=<-8.3008 -0.0357 0.0524>, rot=<-0.0044 0.0485 0.0924 -0.9945>}
 index_02_r {parent=index_01_r, loc=<-3.3587 0.3004 -0.0452>, rot=<-0.002 0.0027 0.3225 -0.9466>}
 index_03_r {parent=index_02_r, loc=<-2.5578 0.123 0.0252>, rot=<-0.002 0.0067 0.2558 -0.9667>}
 middle_metacarpal_r {parent=hand_r, loc=<-2.1856 0.1725 0.5984>, rot=<0.006 0.0168 -0.0329 -0.9993>}
 middle_01_r {parent=middle_metacarpal_r, loc=<-7.3275 0.2447 -0.0129>, rot=<-0.0208 0.0042 0.0858 -0.9961>}
 middle_02_r {parent=middle_01_r, loc=<-4.0123 -0.1401 -0.1175>, rot=<-0.0218 0.0258 0.3512 -0.9357>}
 middle_03_r {parent=middle_02_r, loc=<-2.6077 -0.0534 -0.1373>, rot=<0.0082 0.0188 0.1708 -0.9851>}
 ring_metacarpal_r {parent=hand_r, loc=<-2.2074 0.3625 -1.3638>, rot=<-0.0284 0.0722 -0.0458 -0.9959>}
 ring_01_r {parent=ring_metacarpal_r, loc=<-7.0295 0.6197 0.1446>, rot=<-0.0489 -0.0252 0.1892 -0.9804>}
 ring_02_r {parent=ring_01_r, loc=<-3.6174 -0.4647 0.1027>, rot=<0 0 0.2806 -0.9598>}
 ring_03_r {parent=ring_02_r, loc=<-2.4152 -0.114 -0.0199>, rot=<0 0 0.1189 -0.9929>}
 pinky_metacarpal_r {parent=hand_r, loc=<-1.7616 0.9753 -3.3814>, rot=<-0.0397 0.0814 -0.069 -0.9935>}
 pinky_01_r {parent=pinky_metacarpal_r, loc=<-7.1709 0.4831 0.1094>, rot=<-0.0654 -0.0035 0.2242 -0.9723>}
 pinky_02_r {parent=pinky_01_r, loc=<-2.8274 -0.0893 0.1043>, rot=<0.0037 0.0011 0.2854 -0.9584>}
 pinky_03_r {parent=pinky_02_r, loc=<-2.2046 -0.0307 0.0078>, rot=<-0.0002 0.0098 0.1445 -0.9895>}
 lowerarm_twist_02_r {parent=lowerarm_r, loc=<-10.6707 0 0>, rot=<0 0 0 -1>}
 lowerarm_twist_01_r {parent=lowerarm_r, loc=<-18.316 0 0>, rot=<0 0 0 -1>}
 upperarm_twist_01_r {parent=upperarm_r, loc=<-7.2135 0.1134 0>, rot=<0 0 0 -1>}
 dyn_sleeve_r_base {parent=upperarm_twist_01_r, loc=<0.0007 0.0001 0.0007>, rot=<0.0106 -0.9996 0.0203 0.0171>}
 dyn_sleeve_r_fr {parent=dyn_sleeve_r_base, loc=<-0.9774 -5.0614 0.8551>, rot=<-0.6223 0.0612 -0.0253 -0.7799>}
 dyn_sleeve_r_top {parent=dyn_sleeve_r_base, loc=<-2.63 1.7991 5.9212>, rot=<-0.0109 0.2007 -0.0106 -0.9795>}
 dyn_sleeve_r_bck {parent=dyn_sleeve_r_base, loc=<-1.5736 7.0106 0.0211>, rot=<0.5291 0.1231 -0.0718 -0.8365>}
 dyn_sleeve_r_bot {parent=dyn_sleeve_r_base, loc=<0.5667 0.0956 -6.1968>, rot=<0.0816 -0.1364 0.0026 -0.9873>}
 upperarm_twist_02_r {parent=upperarm_r, loc=<-19.7998 0.1134 0>, rot=<0 0 0 -1>}
 upperarm_r_ctrl_2 {controlType=EULER_TRANSFORM, parent=upperarm_r, type=Control, loc=<0 0 0>, rot=<0 0 0 1>}
 dyn_lapel_l_01 {parent=spine_05, loc=<1.5228 10.0827 -7.6672>, rot=<0.633 -0.497 -0.0606 -0.5904>}
 dyn_lapel_l_02 {parent=dyn_lapel_l_01, loc=<9.3606 0 0>, rot=<0.0057 0.0432 0.1304 -0.9905>}
 dyn_lapel_r_01 {parent=spine_05, loc=<1.5228 10.0827 7.667>, rot=<0.5904 0.0606 0.497 -0.633>}
 dyn_lapel_r_02 {parent=dyn_lapel_r_01, loc=<9.361 0 0>, rot=<-0.0057 0.0432 -0.1304 -0.9905>}
 dyn_lapel_r_01_ctrl {controlType=EULER_TRANSFORM, parent=dyn_lapel_r_01, type=Control, loc=<0 0 0>, rot=<0 0 0 1>}
 dyn_collar_l {parent=spine_05, loc=<12.8072 3.2078 -7.2543>, rot=<0.327 -0.6946 -0.0257 -0.6403>}
 dyn_collar_r {parent=spine_05, loc=<12.0378 3.1459 7.254>, rot=<0.6027 0.0139 0.7608 -0.2401>}
 dyn_necklace_base {parent=spine_05, loc=<7.8076 5.5198 -0.0001>, rot=<0.1555 0.6898 0.6898 0.1555>}
 dyn_necklace {parent=dyn_necklace_base, loc=<0 0 0>, rot=<0 0 0 1>}
 spine_05_ctrl {controlType=EULER_TRANSFORM, parent=spine_05, type=Control, loc=<0 0 0>, rot=<0 0 0 1>}
 thigh_l {parent=pelvis, loc=<-1.0855 -0.2109 -8.7053>, rot=<0.0009 0.0355 0.0155 -0.9993>}
 calf_l {parent=thigh_l, loc=<-42.5284 -0.0019 -0.7297>, rot=<0.0006 -0.0161 -0.0391 -0.9991>}
 foot_l {parent=calf_l, loc=<-46.7401 -2.5014 -0.592>, rot=<0.0427 -0.0238 0.0353 -0.9982>}
 ball_l {parent=foot_l, loc=<-3.5658 10.8066 -0.1544>, rot=<0.0001 -0.0001 -0.7063 -0.7079>}
 calf_twist_02_l {parent=calf_l, loc=<-19.1546 -1.7943 -0.2638>, rot=<0.0237 0.0095 -0.0035 -0.9997>}
 calf_twist_01_l {parent=calf_l, loc=<-35.8483 -1.9197 -0.5794>, rot=<0.0237 0.0095 -0.0035 -0.9997>}
 thigh_twist_01_l {parent=thigh_l, loc=<-27.952 0 0>, rot=<0.0013 0 0 -1>}
 thigh_l_ctrl {controlType=EULER_TRANSFORM, parent=thigh_l, type=Control, loc=<0 0 0>, rot=<0 0 0 1>}
 thigh_r {parent=pelvis, loc=<-1.0855 -0.2109 8.7053>, rot=<0.0355 -0.0009 0.9993 0.0155>}
 calf_r {parent=thigh_r, loc=<42.5284 0.0019 0.7297>, rot=<0.0006 -0.0161 -0.0391 -0.9991>}
 foot_r {parent=calf_r, loc=<46.7401 2.5014 0.592>, rot=<0.0427 -0.0238 0.0353 -0.9982>}
 ball_r {parent=foot_r, loc=<3.5658 -10.8066 0.1544>, rot=<0.0001 -0.0001 -0.7063 -0.7079>}
 calf_twist_02_r {parent=calf_r, loc=<19.1546 1.7943 0.2638>, rot=<0.0237 0.0095 -0.0035 -0.9997>}
 calf_twist_01_r {parent=calf_r, loc=<35.8483 1.9197 0.5794>, rot=<0.0237 0.0095 -0.0035 -0.9997>}
 thigh_twist_01_r {parent=thigh_r, loc=<27.952 0 0>, rot=<0.0013 0 0 -1>}
 dyn_coat_fr_l_base {parent=pelvis, loc=<3.9114 7.9938 -10.7601>, rot=<-0.015 -0.9346 -0.3383 -0.1089>}
 dyn_coat_fr_l_01 {parent=dyn_coat_fr_l_base, loc=<0 0 0>, rot=<0.0002 0.0007 -0.0017 -1>}
 dyn_coat_fr_l_02 {parent=dyn_coat_fr_l_01, loc=<16.0393 0 0>, rot=<0.0034 -0.0052 0.0069 -1>}
 dyn_coat_fr_l_03 {parent=dyn_coat_fr_l_02, loc=<19.2406 0 0>, rot=<0.0436 0.0082 0.0043 -0.999>}
 dyn_coat_fr_l_04 {parent=dyn_coat_fr_l_03, loc=<24.3966 0 0>, rot=<0 0 0 1>}
 dyn_coat_side_l_base {parent=pelvis, loc=<3.4966 -3.5687 -13.8027>, rot=<-0.0037 -0.9766 0.1712 -0.13>}
 dyn_coat_side_l_01 {parent=dyn_coat_side_l_base, loc=<0 0 0>, rot=<-0.0016 -0.0025 0.0097 -0.9999>}
 dyn_coat_side_l_02 {parent=dyn_coat_side_l_01, loc=<16.225 0 0>, rot=<-0.034 -0.0171 0.0085 -0.9992>}
 dyn_coat_side_l_03 {parent=dyn_coat_side_l_02, loc=<19.4207 0 0>, rot=<-0.0047 0.0141 0.0012 -0.9999>}
 dyn_coat_side_l_04 {parent=dyn_coat_side_l_03, loc=<20.508 0 0>, rot=<0 0 0 1>}
 dyn_coat_bck_l_base {parent=pelvis, loc=<3.2332 -11.0309 -5.4809>, rot=<0.044 -0.8014 0.5883 -0.0981>}
 dyn_coat_bck_l_01 {parent=dyn_coat_bck_l_base, loc=<0 0 0>, rot=<0.0005 0.0064 0.0059 -1>}
 dyn_coat_bck_l_02 {parent=dyn_coat_bck_l_01, loc=<16.2274 0 0>, rot=<-0.0124 -0.0155 0.0044 -0.9998>}
 dyn_coat_bck_l_03 {parent=dyn_coat_bck_l_02, loc=<19.3635 0 0>, rot=<-0.0148 -0.0044 0.0037 -0.9999>}
 dyn_coat_bck_l_04 {parent=dyn_coat_bck_l_03, loc=<16.6899 0 0>, rot=<0 0 0 1>}
 dyn_coat_bck_r_base {parent=pelvis, loc=<3.2332 -11.0309 5.481>, rot=<0.0986 -0.5527 0.8255 -0.0577>}
 dyn_coat_bck_r_01 {parent=dyn_coat_bck_r_base, loc=<0 0 0>, rot=<-0.0001 -0.0016 0 -1>}
 dyn_coat_bck_r_02 {parent=dyn_coat_bck_r_01, loc=<16.2275 0 0>, rot=<0.0124 -0.0158 -0.0031 -0.9998>}
 dyn_coat_bck_r_03 {parent=dyn_coat_bck_r_02, loc=<19.3635 0 0>, rot=<0.0149 -0.0047 -0.0034 -0.9999>}
 dyn_coat_bck_r_04 {parent=dyn_coat_bck_r_03, loc=<16.6898 0 0>, rot=<0 0 0 1>}
 dyn_coat_side_r_base {parent=pelvis, loc=<3.4966 -3.5687 13.803>, rot=<0.1245 -0.0887 0.9881 -0.0161>}
 dyn_coat_side_r_01 {parent=dyn_coat_side_r_base, loc=<0 0 0>, rot=<-0.0004 0.0004 0 -1>}
 dyn_coat_side_r_02 {parent=dyn_coat_side_r_01, loc=<16.2249 0 0>, rot=<0.034 -0.0183 -0.0054 -0.9992>}
 dyn_coat_side_r_03 {parent=dyn_coat_side_r_02, loc=<19.4208 0 0>, rot=<0.0048 0.0137 -0.0036 -0.9999>}
 dyn_coat_side_r_04 {parent=dyn_coat_side_r_03, loc=<20.508 0 0>, rot=<0 0 0 1>}
 dyn_coat_fr_r_base {parent=pelvis, loc=<3.9114 7.9938 10.76>, rot=<0.1116 0.4184 0.9014 0.007>}
 dyn_coat_fr_r_01 {parent=dyn_coat_fr_r_base, loc=<0 0 0>, rot=<-0.0004 -0.0014 0.0009 -1>}
 dyn_coat_fr_r_02 {parent=dyn_coat_fr_r_01, loc=<16.0394 0 0>, rot=<0.0401 -0.006 -0.0061 -0.9992>}
 dyn_coat_fr_r_03 {parent=dyn_coat_fr_r_02, loc=<19.2405 0 0>, rot=<0 0.0081 -0.0047 -1>}
 dyn_coat_fr_r_04 {parent=dyn_coat_fr_r_03, loc=<24.3966 0 0>, rot=<0 0 0 1>}
 root_ctrl {controlType=EULER_TRANSFORM, parent=root, type=Control, loc=<0 0 0>, rot=<0 0 0 1>}
 upperarm_r_ctrl {controlType=EULER_TRANSFORM, type=Control, loc=<0 0 0>, rot=<0 0 0 1>}
]
variables [
 Is Crouching {type=bool, default=False}
 Is Playing Emote {type=bool, default=False}
 Is Targeting {type=bool, default=False}
 L Arm Out Up {type=double, default=0}
 L Calf Run {type=double, default=0}
 L Leg Back {type=double, default=0}
 L Leg Fwd {type=double, default=0}
 L Leg Fwd_EXTREME {type=double, default=0}
 L Leg Left {type=double, default=0}
 L Leg Out {type=double, default=0}
 L Thigh Run {type=double, default=0}
 R Arm Back {type=double, default=0}
 R Arm Out Up {type=double, default=0}
 R Arm Up {type=double, default=0}
 R Arm Up_Full {type=double, default=0}
 R Calf Run {type=double, default=0}
 R Lapel Moving {type=double, default=0}
 R Leg Back {type=double, default=0}
 R Leg Fwd {type=double, default=0}
 R Leg Right {type=double, default=0}
 R Thigh Run {type=double, default=0}
]
