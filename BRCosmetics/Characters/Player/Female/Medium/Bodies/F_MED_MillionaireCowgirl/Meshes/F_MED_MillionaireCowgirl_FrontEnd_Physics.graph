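# Physics asset for the front-end mesh of the MillionaireCowgirl female player body, built to simulate her long coat.
# Six coat chains (left and right front, side and back) hang from the pelvis, each a base plus four boxes, alongside capsules on pelvis, thighs, calves and feet.
# Chain links allow narrow swing1, wider swing2 and zero twist, while pelvis and leg joints rotate freely.
# Two opposing free-rotation constraints tie the back tips together.
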PhysicsAsset /BRCosmetics/Characters/Player/Female/Medium/Bodies/F_MED_MillionaireCowgirl/Meshes/F_MED_MillionaireCowgirl_FrontEnd_Physics
bodies {
 calf_l [{shapeType=Capsule, length=26.5353, radius=8.9647, shapeName=calf_l_capsule, center=<-21.8857 -2.7832 0>, rotation=<rot 0 89.9999 90>}]
 calf_r [{shapeType=Capsule, length=26.5353, radius=8.9647, shapeName=calf_r_capsule, center=<21.7111 1.7927 0>, rotation=<rot 0 89.9994 -90>}]
 dyn_coat_bck_l_01 [{shapeType=Box, extentX=14.5, extentY=23.5, extentZ=5.3432, shapeName=dyn_coat_bck_l_01_box, center=<9.7827 -0.0101 5>}]
 dyn_coat_bck_l_02 [{shapeType=Box, extentX=17.5, extentY=20.5, extentZ=5.3432, shapeName=dyn_coat_bck_l_02_box, center=<9.7827 -0.0101 5>}]
 dyn_coat_bck_l_03 [{shapeType=Box, extentX=24.5, extentY=27.5, extentZ=7.3605, shapeName=dyn_coat_bck_l_03_box, center=<9.7826 -0.0101 5>}]
 dyn_coat_bck_l_04 [{shapeType=Box, extentX=29.5, extentY=51.5, extentZ=9.6259, shapeName=dyn_coat_bck_l_04_box, center=<5.1755 -2.5421 5>}]
 dyn_coat_bck_l_base [{shapeType=Box, extentX=6.5, extentY=8.5, extentZ=5.5, shapeName=dyn_coat_bck_l_base_box, center=<9.7826 0 -1.1405>}]
 dyn_coat_bck_r_01 [{shapeType=Box, extentX=14.5, extentY=23.5, extentZ=5.3432, shapeName=dyn_coat_bck_r_01_box, center=<9.7827 -0.0101 5>}]
 dyn_coat_bck_r_02 [{shapeType=Box, extentX=15.5, extentY=17.5, extentZ=5.3432, shapeName=dyn_coat_bck_r_02_box, center=<10.0018 -0.0101 5>}]
 dyn_coat_bck_r_03 [{shapeType=Box, extentX=24.5, extentY=27.5, extentZ=7.3605, shapeName=dyn_coat_bck_r_03_box, center=<9.4443 -0.0101 5>}]
 dyn_coat_bck_r_04 [{shapeType=Box, extentX=31.5, extentY=54.5, extentZ=9.6259, shapeName=dyn_coat_bck_r_04_box, center=<5.0481 -2.5421 5>}]
 dyn_coat_bck_r_base [{shapeType=Box, extentX=6.5, extentY=8.5, extentZ=5.5, shapeName=dyn_coat_bck_r_base_box, center=<9.7826 0 -1.1405>}]
 dyn_coat_fr_l_01 [{
   shapeType=Box
   extentX=14.5
   extentY=23.5
   extentZ=7
   shapeName=dyn_coat_fr_l_01_box
   center=<9.7826 -0.0101 5.3989>
   rotation=<rot -9.9998 0.0003 -0.0001>
  }]
 dyn_coat_fr_l_02 [{shapeType=Box, extentX=17.5, extentY=23.5, extentZ=5.3432, shapeName=dyn_coat_fr_l_02_box, center=<9.7826 -0.1846 5>}]
 dyn_coat_fr_l_03 [{shapeType=Box, extentX=17.5, extentY=23.5, extentZ=5.3432, shapeName=dyn_coat_fr_l_03_box, center=<9.7826 -0.0101 5>}]
 dyn_coat_fr_l_04 [{shapeType=Box, extentX=17.5, extentY=21.5, extentZ=5.3432, shapeName=dyn_coat_fr_l_04_box, center=<13.7651 -0.0101 5>}]
 dyn_coat_fr_l_base [{shapeType=Box, extentX=6.5, extentY=8.5, extentZ=5.5, shapeName=dyn_coat_fr_l_base_box, center=<9.7826 0 -1.1405>}]
 dyn_coat_fr_r_01 [{
   shapeType=Box
   extentX=14.5
   extentY=23.5
   extentZ=6
   shapeName=dyn_coat_fr_r_01_box
   center=<9.7826 -0.0101 5.3989>
   rotation=<rot -9.9997 -0.0002 0.0001>
  }]
 dyn_coat_fr_r_02 [{shapeType=Box, extentX=17.5, extentY=23.5, extentZ=5.3432, shapeName=dyn_coat_fr_r_02_box, center=<9.7826 -0.0101 5>}]
 dyn_coat_fr_r_03 [{shapeType=Box, extentX=17.5, extentY=23.5, extentZ=5.3432, shapeName=dyn_coat_fr_r_03_box, center=<9.7826 -0.0101 5>}]
 dyn_coat_fr_r_04 [{shapeType=Box, extentX=22.5, extentY=23.5, extentZ=5.3432, shapeName=dyn_coat_fr_r_04_box, center=<11.7473 -0.0101 5>}]
 dyn_coat_fr_r_base [{shapeType=Box, extentX=6.5, extentY=8.5, extentZ=5.5, shapeName=dyn_coat_fr_r_base_box, center=<9.7826 0 -1.1405>}]
 dyn_coat_side_l_01 [{shapeType=Box, extentX=14.5, extentY=23.5, extentZ=5.3432, shapeName=dyn_coat_side_l_01_box, center=<9.7826 -0.0101 3>}]
 dyn_coat_side_l_02 [{shapeType=Box, extentX=17.5, extentY=23.5, extentZ=5.3432, shapeName=dyn_coat_side_l_02_box, center=<9.7826 0.3344 3>}]
 dyn_coat_side_l_03 [{shapeType=Box, extentX=24.5, extentY=29.5, extentZ=5.3432, shapeName=dyn_coat_side_l_03_box, center=<12.2603 -0.0101 3>}]
 dyn_coat_side_l_04 [{shapeType=Box, extentX=23.5, extentY=47.5, extentZ=5.3432, shapeName=dyn_coat_side_l_04_box, center=<8.8854 -2.5647 3>}]
 dyn_coat_side_l_base [{shapeType=Box, extentX=6.5, extentY=8.5, extentZ=5.5, shapeName=dyn_coat_side_l_base_box, center=<9.7826 0 -1.1405>}]
 dyn_coat_side_r_01 [{shapeType=Box, extentX=14.5, extentY=23.5, extentZ=5.3432, shapeName=dyn_coat_side_r_01_box, center=<9.7826 -0.0101 5>}]
 dyn_coat_side_r_02 [{shapeType=Box, extentX=15.5, extentY=16.5, extentZ=5.3432, shapeName=dyn_coat_side_r_02_box, center=<9.7826 -0.0101 5>}]
 dyn_coat_side_r_03 [{shapeType=Box, extentX=20.5, extentY=23.5, extentZ=5.3432, shapeName=dyn_coat_side_r_03_box, center=<10.0649 -0.0101 5>}]
 dyn_coat_side_r_04 [{shapeType=Box, extentX=19.5, extentY=36.5, extentZ=5.3432, shapeName=dyn_coat_side_r_04_box, center=<9.1339 0.5026 5>}]
 dyn_coat_side_r_base [{shapeType=Box, extentX=6.5, extentY=8.5, extentZ=5.5, shapeName=dyn_coat_side_r_base_box, center=<9.7826 0 -1.1405>}]
 foot_l [{shapeType=Capsule, length=9, radius=11.5, shapeName=foot_l_capsule, center=<-3.1667 5.4849 0>, rotation=<rot 0 29.9996 -90>}]
 foot_r [{shapeType=Capsule, length=9.5, radius=10, shapeName=foot_r_capsule, center=<2.0656 -3.5778 0>, rotation=<rot 0 29.9996 -90>}]
 pelvis [{shapeType=Capsule, length=1, radius=3, shapeName=pelvis_capsule}]
 thigh_l [{
   shapeType=Capsule
   length=43.0702
   radius=6.9552
   shapeName=thigh_l_capsule
   center=<-24.3205 -0.2407 -0.7814>
   rotation=<rot 90 180 -0.5669>
  }]
 thigh_r [{
   shapeType=Capsule
   length=43.0702
   radius=6.9552
   shapeName=thigh_r_capsule
   center=<24.3217 -0.0001 0.7265>
   rotation=<rot 90 -179.9999 -179.9996>
  }]
}
constraints [
 {
  bone1Name=calf_l
  bone2Name=thigh_l
  swing1LimitDegrees=45
  swing1Motion=Free
  swing2LimitDegrees=45
  swing2Motion=Free
  twistLimitDegrees=45
  twistMotion=Free
 }
 {
  bone1Name=calf_r
  bone2Name=thigh_r
  swing1LimitDegrees=45
  swing1Motion=Free
  swing2LimitDegrees=45
  swing2Motion=Free
  twistLimitDegrees=45
  twistMotion=Free
 }
 {
  bone1Name=foot_r
  bone2Name=calf_r
  swing1LimitDegrees=45
  swing1Motion=Free
  swing2LimitDegrees=45
  swing2Motion=Free
  twistLimitDegrees=45
  twistMotion=Free
 }
 {
  bone1Name=foot_l
  bone2Name=calf_l
  swing1LimitDegrees=45
  swing1Motion=Free
  swing2LimitDegrees=45
  swing2Motion=Free
  twistLimitDegrees=45
  twistMotion=Free
 }
 {
  bone1Name=thigh_l
  bone2Name=pelvis
  swing1LimitDegrees=45
  swing1Motion=Free
  swing2LimitDegrees=45
  swing2Motion=Free
  twistLimitDegrees=45
  twistMotion=Free
 }
 {
  bone1Name=thigh_r
  bone2Name=pelvis
  swing1LimitDegrees=45
  swing1Motion=Free
  swing2LimitDegrees=45
  swing2Motion=Free
  twistLimitDegrees=45
  twistMotion=Free
 }
 {
  bone1Name=dyn_coat_bck_l_base
  bone2Name=pelvis
  swing1LimitDegrees=45
  swing1Motion=Free
  swing2LimitDegrees=45
  swing2Motion=Free
  twistLimitDegrees=45
  twistMotion=Free
 }
 {bone1Name=dyn_coat_bck_l_01, bone2Name=dyn_coat_bck_l_base, swing1LimitDegrees=5, swing2LimitDegrees=20, twistLimitDegrees=0}
 {bone1Name=dyn_coat_bck_l_02, bone2Name=dyn_coat_bck_l_01, swing1LimitDegrees=5, swing2LimitDegrees=20, twistLimitDegrees=0}
 {bone1Name=dyn_coat_bck_l_03, bone2Name=dyn_coat_bck_l_02, swing1LimitDegrees=5, swing2LimitDegrees=25, twistLimitDegrees=0}
 {bone1Name=dyn_coat_bck_l_04, bone2Name=dyn_coat_bck_l_03, swing1LimitDegrees=5, swing2LimitDegrees=20, twistLimitDegrees=0}
 {
  bone1Name=dyn_coat_bck_r_base
  bone2Name=pelvis
  swing1LimitDegrees=45
  swing1Motion=Free
  swing2LimitDegrees=45
  swing2Motion=Free
  twistLimitDegrees=45
  twistMotion=Free
 }
 {bone1Name=dyn_coat_bck_r_01, bone2Name=dyn_coat_bck_r_base, swing1LimitDegrees=5, swing2LimitDegrees=20, twistLimitDegrees=0}
 {bone1Name=dyn_coat_bck_r_02, bone2Name=dyn_coat_bck_r_01, swing1LimitDegrees=5, swing2LimitDegrees=20, twistLimitDegrees=0}
 {bone1Name=dyn_coat_bck_r_03, bone2Name=dyn_coat_bck_r_02, swing1LimitDegrees=5, swing2LimitDegrees=25, twistLimitDegrees=0}
 {bone1Name=dyn_coat_bck_r_04, bone2Name=dyn_coat_bck_r_03, swing1LimitDegrees=5, swing2LimitDegrees=20, twistLimitDegrees=0}
 {
  bone1Name=dyn_coat_fr_l_base
  bone2Name=pelvis
  swing1LimitDegrees=45
  swing1Motion=Free
  swing2LimitDegrees=45
  swing2Motion=Free
  twistLimitDegrees=45
  twistMotion=Free
 }
 {bone1Name=dyn_coat_fr_l_01, bone2Name=dyn_coat_fr_l_base, swing1LimitDegrees=5, swing2LimitDegrees=20, twistLimitDegrees=0}
 {bone1Name=dyn_coat_fr_l_02, bone2Name=dyn_coat_fr_l_01, swing1LimitDegrees=5, swing2LimitDegrees=20, twistLimitDegrees=0}
 {bone1Name=dyn_coat_fr_l_03, bone2Name=dyn_coat_fr_l_02, swing1LimitDegrees=5, swing2LimitDegrees=30, twistLimitDegrees=0}
 {bone1Name=dyn_coat_fr_l_04, bone2Name=dyn_coat_fr_l_03, swing1LimitDegrees=5, swing2LimitDegrees=20, twistLimitDegrees=0}
 {
  bone1Name=dyn_coat_fr_r_base
  bone2Name=pelvis
  swing1LimitDegrees=45
  swing1Motion=Free
  swing2LimitDegrees=45
  swing2Motion=Free
  twistLimitDegrees=45
  twistMotion=Free
 }
 {bone1Name=dyn_coat_fr_r_01, bone2Name=dyn_coat_fr_r_base, swing1LimitDegrees=5, swing2LimitDegrees=20, twistLimitDegrees=0}
 {bone1Name=dyn_coat_fr_r_02, bone2Name=dyn_coat_fr_r_01, swing1LimitDegrees=5, swing2LimitDegrees=20, twistLimitDegrees=0}
 {bone1Name=dyn_coat_fr_r_03, bone2Name=dyn_coat_fr_r_02, swing1LimitDegrees=5, swing2LimitDegrees=20, twistLimitDegrees=0}
 {bone1Name=dyn_coat_fr_r_04, bone2Name=dyn_coat_fr_r_03, swing1LimitDegrees=5, swing2LimitDegrees=20, twistLimitDegrees=0}
 {
  bone1Name=dyn_coat_side_l_base
  bone2Name=pelvis
  swing1LimitDegrees=45
  swing1Motion=Free
  swing2LimitDegrees=45
  swing2Motion=Free
  twistLimitDegrees=45
  twistMotion=Free
 }
 {bone1Name=dyn_coat_side_l_01, bone2Name=dyn_coat_side_l_base, swing1LimitDegrees=5, swing2LimitDegrees=20, twistLimitDegrees=0}
 {bone1Name=dyn_coat_side_l_02, bone2Name=dyn_coat_side_l_01, swing1LimitDegrees=5, swing2LimitDegrees=20, twistLimitDegrees=0}
 {bone1Name=dyn_coat_side_l_03, bone2Name=dyn_coat_side_l_02, swing1LimitDegrees=5, swing2LimitDegrees=25, twistLimitDegrees=0}
 {bone1Name=dyn_coat_side_l_04, bone2Name=dyn_coat_side_l_03, swing1LimitDegrees=5, swing2LimitDegrees=20, twistLimitDegrees=0}
 {
  bone1Name=dyn_coat_side_r_base
  bone2Name=pelvis
  swing1LimitDegrees=45
  swing1Motion=Free
  swing2LimitDegrees=45
  swing2Motion=Free
  twistLimitDegrees=45
  twistMotion=Free
 }
 {bone1Name=dyn_coat_side_r_01, bone2Name=dyn_coat_side_r_base, swing1LimitDegrees=5, swing2LimitDegrees=20, twistLimitDegrees=0}
 {bone1Name=dyn_coat_side_r_02, bone2Name=dyn_coat_side_r_01, swing1LimitDegrees=5, swing2LimitDegrees=20, twistLimitDegrees=0}
 {bone1Name=dyn_coat_side_r_03, bone2Name=dyn_coat_side_r_02, swing1LimitDegrees=5, swing2LimitDegrees=25, twistLimitDegrees=0}
 {bone1Name=dyn_coat_side_r_04, bone2Name=dyn_coat_side_r_03, swing1LimitDegrees=5, swing2LimitDegrees=20, twistLimitDegrees=0}
 {
  bone1Name=dyn_coat_bck_l_04
  bone2Name=dyn_coat_bck_r_04
  swing1LimitDegrees=45
  swing1Motion=Free
  swing2LimitDegrees=45
  swing2Motion=Free
  twistLimitDegrees=45
  twistMotion=Free
 }
 {
  bone1Name=dyn_coat_bck_r_04
  bone2Name=dyn_coat_bck_l_04
  swing1LimitDegrees=45
  swing1Motion=Free
  swing2LimitDegrees=45
  swing2Motion=Free
  twistLimitDegrees=45
  twistMotion=Free
 }
]
