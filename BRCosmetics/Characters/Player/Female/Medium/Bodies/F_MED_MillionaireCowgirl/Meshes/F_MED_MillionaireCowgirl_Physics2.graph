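# Physics asset for the Millionaire Cowgirl female player body, setting up collision and joints for the coat's dynamic bones.
# Six box-body chains (front, side, back; left and right) hang from pelvis, with limited swing and zero twist below free-rotating base joints.
# Leg and foot capsules supply body collision, while a huge zero-length root capsule acts as a near-flat floor.
# One extra free-angular constraint links right back and side panels.
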PhysicsAsset /BRCosmetics/Characters/Player/Female/Medium/Bodies/F_MED_MillionaireCowgirl/Meshes/F_MED_MillionaireCowgirl_Physics2
bodies {
 calf_l [{shapeType=Capsule, length=27.5, radius=8, shapeName=calf_l_capsule, center=<-21.8857 -2.7832 0>, rotation=<rot 0 89.9999 90>}]
 calf_r [{shapeType=Capsule, length=27.5, radius=8, shapeName=calf_r_capsule, center=<21.7111 1.7927 0>, rotation=<rot 0 89.9994 -90>}]
 dyn_coat_bck_l_01 [{shapeType=Box, extentX=14.5, extentY=23.5, extentZ=3.5, shapeName=dyn_coat_bck_l_01_box, center=<9.7826 0 -0.0545>}]
 dyn_coat_bck_l_02 [{shapeType=Box, extentX=17.5, extentY=23.5, extentZ=5.5, shapeName=dyn_coat_bck_l_02_box, center=<9.7826 0 0.4216>}]
 dyn_coat_bck_l_03 [{shapeType=Box, extentX=24.5, extentY=27.5, extentZ=10.5, shapeName=dyn_coat_bck_l_03_box, center=<9.7826 0 -1.1405>}]
 dyn_coat_bck_l_04 [{
   shapeType=Box
   extentX=29.5
   extentY=57.5
   extentZ=8.5
   shapeName=dyn_coat_bck_l_04_box
   center=<6.1626 0 -1.8629>
   rotation=<rot 9.9999 0 0>
  }]
 dyn_coat_bck_l_base [{shapeType=Box, extentX=6.5, extentY=8.5, extentZ=5.5, shapeName=dyn_coat_bck_l_base_box, center=<9.7826 0 -1.1405>}]
 dyn_coat_bck_r_01 [{shapeType=Box, extentX=14.5, extentY=23.5, extentZ=3.5, shapeName=dyn_coat_bck_r_01_box, center=<9.7826 0 -0.0545>}]
 dyn_coat_bck_r_02 [{shapeType=Box, extentX=17.5, extentY=23.5, extentZ=5.5, shapeName=dyn_coat_bck_r_02_box, center=<9.7826 0 0.4216>}]
 dyn_coat_bck_r_03 [{shapeType=Box, extentX=24.5, extentY=27.5, extentZ=10.5, shapeName=dyn_coat_bck_r_03_box, center=<9.7826 0 -1.1405>}]
 dyn_coat_bck_r_04 [{
   shapeType=Box
   extentX=31.5
   extentY=54.5
   extentZ=8.5
   shapeName=dyn_coat_bck_r_04_box
   center=<6.0352 0 -2.5649>
   rotation=<rot 9.9999 0 0>
  }]
 dyn_coat_bck_r_base [{shapeType=Box, extentX=6.5, extentY=8.5, extentZ=5.5, shapeName=dyn_coat_bck_r_base_box, center=<9.7826 0 -1.1405>}]
 dyn_coat_fr_l_01 [{shapeType=Box, extentX=14.5, extentY=23.5, extentZ=6.5, shapeName=dyn_coat_fr_l_01_box, center=<9.7826 0 -0.0545>}]
 dyn_coat_fr_l_02 [{shapeType=Box, extentX=17.5, extentY=23.5, extentZ=5.5, shapeName=dyn_coat_fr_l_02_box, center=<9.7826 0 0.4216>}]
 dyn_coat_fr_l_03 [{shapeType=Box, extentX=17.5, extentY=23.5, extentZ=5.5, shapeName=dyn_coat_fr_l_03_box, center=<9.7826 0 -1.1405>}]
 dyn_coat_fr_l_04 [{shapeType=Box, extentX=17.5, extentY=21.5, extentZ=6.5, shapeName=dyn_coat_fr_l_04_box, center=<14.4138 0 -0.7847>, rotation=<rot 10 0 0>}]
 dyn_coat_fr_l_base [{shapeType=Box, extentX=6.5, extentY=8.5, extentZ=5.5, shapeName=dyn_coat_fr_l_base_box, center=<9.7826 0 -1.1405>}]
 dyn_coat_fr_r_01 [{shapeType=Box, extentX=14.5, extentY=23.5, extentZ=5.5, shapeName=dyn_coat_fr_r_01_box, center=<9.7826 0 -0.0545>}]
 dyn_coat_fr_r_02 [{shapeType=Box, extentX=17.5, extentY=23.5, extentZ=5.5, shapeName=dyn_coat_fr_r_02_box, center=<9.7826 0 0.4216>}]
 dyn_coat_fr_r_03 [{shapeType=Box, extentX=17.5, extentY=23.5, extentZ=5.5, shapeName=dyn_coat_fr_r_03_box, center=<9.7826 0 -1.1405>}]
 dyn_coat_fr_r_04 [{
   shapeType=Box
   extentX=17.5
   extentY=22.5
   extentZ=6.5
   shapeName=dyn_coat_fr_r_04_box
   center=<12.396 0 -1.1405>
   rotation=<rot 10.0001 -0.0001 0>
  }]
 dyn_coat_fr_r_base [{shapeType=Box, extentX=6.5, extentY=8.5, extentZ=5.5, shapeName=dyn_coat_fr_r_base_box, center=<9.7826 0 -1.1405>}]
 dyn_coat_side_l_01 [{shapeType=Box, extentX=14.5, extentY=23.5, extentZ=4.5, shapeName=dyn_coat_side_l_01_box, center=<9.7826 0 -0.0545>}]
 dyn_coat_side_l_02 [{shapeType=Box, extentX=17.5, extentY=23.5, extentZ=5.5, shapeName=dyn_coat_side_l_02_box, center=<9.7826 0 0.4216>}]
 dyn_coat_side_l_03 [{shapeType=Box, extentX=24.5, extentY=29.5, extentZ=7.5, shapeName=dyn_coat_side_l_03_box, center=<12.2603 0 -2.9109>}]
 dyn_coat_side_l_04 [{
   shapeType=Box
   extentX=23.5
   extentY=47.5
   extentZ=7.5
   shapeName=dyn_coat_side_l_04_box
   center=<9.5341 -1.829 -3.1259>
   rotation=<rot 9.8464 -1.7535 -10.1512>
  }]
 dyn_coat_side_l_base [{shapeType=Box, extentX=6.5, extentY=8.5, extentZ=5.5, shapeName=dyn_coat_side_l_base_box, center=<9.7826 0 -1.1405>}]
 dyn_coat_side_r_01 [{shapeType=Box, extentX=14.5, extentY=23.5, extentZ=5.5, shapeName=dyn_coat_side_r_01_box, center=<9.7826 0 -0.0545>}]
 dyn_coat_side_r_02 [{shapeType=Box, extentX=17.5, extentY=23.5, extentZ=5.5, shapeName=dyn_coat_side_r_02_box, center=<9.7826 0 0.4216>}]
 dyn_coat_side_r_03 [{shapeType=Box, extentX=20.5, extentY=31.5, extentZ=5.5, shapeName=dyn_coat_side_r_03_box, center=<10.0649 0 -1.7644>}]
 dyn_coat_side_r_04 [{
   shapeType=Box
   extentX=19.5
   extentY=36.5
   extentZ=3.5
   shapeName=dyn_coat_side_r_04_box
   center=<9.7826 0.2807 -1.9119>
   rotation=<rot 9.8465 1.7535 10.1511>
  }]
 dyn_coat_side_r_base [{shapeType=Box, extentX=6.5, extentY=8.5, extentZ=5.5, shapeName=dyn_coat_side_r_base_box, center=<9.7826 0 -1.1405>}]
 foot_l [{shapeType=Capsule, length=4, radius=8, shapeName=foot_l_capsule, center=<-3.1667 5.4849 0>, rotation=<rot 0 29.9996 -90>}]
 foot_r [{shapeType=Capsule, length=4, radius=8, shapeName=foot_r_capsule, center=<2.0656 -3.5778 0>, rotation=<rot 0 29.9996 -90>}]
 pelvis [{shapeType=Capsule, length=1, radius=3, shapeName=pelvis_capsule}]
 root [{shapeType=Capsule, length=0, radius=289.5, shapeName=root_capsule, center=<0 0 -286.8952>}]
 thigh_l [{shapeType=Capsule, length=27.5, radius=7, shapeName=thigh_l_capsule, center=<-23.673 -0.2343 0>, rotation=<rot 90 180 -0.5669>}]
 thigh_r [{shapeType=Capsule, length=27.5, radius=7, shapeName=thigh_r_capsule, center=<23.6742 -0.0001 0>, rotation=<rot 90 -0.0002 0.0001>}]
}
constraints [
 {
  bone1Name=calf_l
  bone2Name=thigh_l
  swing1LimitDegrees=45
  swing1Motion=Free
  swing2LimitDegrees=45
  swing2Motion=Free
  twistLimitDegrees=45
  twistMotion=Free
 }
 {
  bone1Name=calf_r
  bone2Name=thigh_r
  swing1LimitDegrees=45
  swing1Motion=Free
  swing2LimitDegrees=45
  swing2Motion=Free
  twistLimitDegrees=45
  twistMotion=Free
 }
 {
  bone1Name=foot_r
  bone2Name=calf_r
  swing1LimitDegrees=45
  swing1Motion=Free
  swing2LimitDegrees=45
  swing2Motion=Free
  twistLimitDegrees=45
  twistMotion=Free
 }
 {
  bone1Name=foot_l
  bone2Name=calf_l
  swing1LimitDegrees=45
  swing1Motion=Free
  swing2LimitDegrees=45
  swing2Motion=Free
  twistLimitDegrees=45
  twistMotion=Free
 }
 {
  bone1Name=thigh_l
  bone2Name=pelvis
  swing1LimitDegrees=45
  swing1Motion=Free
  swing2LimitDegrees=45
  swing2Motion=Free
  twistLimitDegrees=45
  twistMotion=Free
 }
 {
  bone1Name=thigh_r
  bone2Name=pelvis
  swing1LimitDegrees=45
  swing1Motion=Free
  swing2LimitDegrees=45
  swing2Motion=Free
  twistLimitDegrees=45
  twistMotion=Free
 }
 {
  bone1Name=dyn_coat_bck_l_base
  bone2Name=pelvis
  swing1LimitDegrees=45
  swing1Motion=Free
  swing2LimitDegrees=45
  swing2Motion=Free
  twistLimitDegrees=45
  twistMotion=Free
 }
 {bone1Name=dyn_coat_bck_l_01, bone2Name=dyn_coat_bck_l_base, swing1LimitDegrees=5, swing2LimitDegrees=30, twistLimitDegrees=0}
 {bone1Name=dyn_coat_bck_l_02, bone2Name=dyn_coat_bck_l_01, swing1LimitDegrees=5, swing2LimitDegrees=30, twistLimitDegrees=0}
 {bone1Name=dyn_coat_bck_l_03, bone2Name=dyn_coat_bck_l_02, swing1LimitDegrees=5, swing2LimitDegrees=50, twistLimitDegrees=0}
 {bone1Name=dyn_coat_bck_l_04, bone2Name=dyn_coat_bck_l_03, swing1LimitDegrees=5, swing2LimitDegrees=35, twistLimitDegrees=0}
 {
  bone1Name=dyn_coat_bck_r_base
  bone2Name=pelvis
  swing1LimitDegrees=45
  swing1Motion=Free
  swing2LimitDegrees=45
  swing2Motion=Free
  twistLimitDegrees=45
  twistMotion=Free
 }
 {bone1Name=dyn_coat_bck_r_01, bone2Name=dyn_coat_bck_r_base, swing1LimitDegrees=5, swing2LimitDegrees=20, twistLimitDegrees=0}
 {bone1Name=dyn_coat_bck_r_02, bone2Name=dyn_coat_bck_r_01, swing1LimitDegrees=5, swing2LimitDegrees=30, twistLimitDegrees=0}
 {bone1Name=dyn_coat_bck_r_03, bone2Name=dyn_coat_bck_r_02, swing1LimitDegrees=5, swing2LimitDegrees=50, twistLimitDegrees=0}
 {
  bone1Name=dyn_coat_fr_l_base
  bone2Name=pelvis
  swing1LimitDegrees=45
  swing1Motion=Free
  swing2LimitDegrees=45
  swing2Motion=Free
  twistLimitDegrees=45
  twistMotion=Free
 }
 {bone1Name=dyn_coat_fr_l_01, bone2Name=dyn_coat_fr_l_base, swing1LimitDegrees=10, swing2LimitDegrees=20, twistLimitDegrees=0}
 {bone1Name=dyn_coat_fr_l_02, bone2Name=dyn_coat_fr_l_01, swing1LimitDegrees=10, swing2LimitDegrees=40, twistLimitDegrees=0}
 {bone1Name=dyn_coat_fr_l_03, bone2Name=dyn_coat_fr_l_02, swing1LimitDegrees=10, swing2LimitDegrees=40, twistLimitDegrees=0}
 {
  bone1Name=dyn_coat_fr_r_base
  bone2Name=pelvis
  swing1LimitDegrees=45
  swing1Motion=Free
  swing2LimitDegrees=45
  swing2Motion=Free
  twistLimitDegrees=45
  twistMotion=Free
 }
 {bone1Name=dyn_coat_fr_r_01, bone2Name=dyn_coat_fr_r_base, swing1LimitDegrees=10, swing2LimitDegrees=20, twistLimitDegrees=0}
 {bone1Name=dyn_coat_fr_r_02, bone2Name=dyn_coat_fr_r_01, swing1LimitDegrees=5, swing2LimitDegrees=20, twistLimitDegrees=0}
 {bone1Name=dyn_coat_fr_r_03, bone2Name=dyn_coat_fr_r_02, swing1LimitDegrees=0, swing2LimitDegrees=40, twistLimitDegrees=0}
 {bone1Name=dyn_coat_fr_r_04, bone2Name=dyn_coat_fr_r_03, swing1LimitDegrees=0, swing2LimitDegrees=20, twistLimitDegrees=0}
 {
  bone1Name=dyn_coat_side_l_base
  bone2Name=pelvis
  swing1LimitDegrees=45
  swing1Motion=Free
  swing2LimitDegrees=45
  swing2Motion=Free
  twistLimitDegrees=45
  twistMotion=Free
 }
 {bone1Name=dyn_coat_side_l_01, bone2Name=dyn_coat_side_l_base, swing1LimitDegrees=10, swing2LimitDegrees=20, twistLimitDegrees=0}
 {bone1Name=dyn_coat_side_l_02, bone2Name=dyn_coat_side_l_01, swing1LimitDegrees=10, swing2LimitDegrees=40, twistLimitDegrees=0}
 {bone1Name=dyn_coat_side_l_03, bone2Name=dyn_coat_side_l_02, swing1LimitDegrees=10, swing2LimitDegrees=40, twistLimitDegrees=0}
 {bone1Name=dyn_coat_side_l_04, bone2Name=dyn_coat_side_l_03, swing1LimitDegrees=10, swing2LimitDegrees=40, twistLimitDegrees=0}
 {
  bone1Name=dyn_coat_side_r_base
  bone2Name=pelvis
  swing1LimitDegrees=45
  swing1Motion=Free
  swing2LimitDegrees=45
  swing2Motion=Free
  twistLimitDegrees=45
  twistMotion=Free
 }
 {bone1Name=dyn_coat_side_r_01, bone2Name=dyn_coat_side_r_base, swing1LimitDegrees=10, swing2LimitDegrees=20, twistLimitDegrees=0}
 {bone1Name=dyn_coat_side_r_02, bone2Name=dyn_coat_side_r_01, swing1LimitDegrees=5, swing2LimitDegrees=35, twistLimitDegrees=0}
 {bone1Name=dyn_coat_side_r_03, bone2Name=dyn_coat_side_r_02, swing1LimitDegrees=5, swing2LimitDegrees=30, twistLimitDegrees=0}
 {bone1Name=dyn_coat_side_r_04, bone2Name=dyn_coat_side_r_03, swing1LimitDegrees=5, swing2LimitDegrees=35, twistLimitDegrees=0}
 {bone1Name=dyn_coat_fr_l_04, bone2Name=dyn_coat_fr_l_03, swing1LimitDegrees=10, swing2LimitDegrees=40, twistLimitDegrees=0}
 {bone1Name=dyn_coat_bck_r_04, bone2Name=dyn_coat_bck_r_03, swing1LimitDegrees=5, swing2LimitDegrees=35, twistLimitDegrees=0}
 {
  bone1Name=dyn_coat_bck_r_03
  bone2Name=dyn_coat_side_r_03
  swing1LimitDegrees=45
  swing1Motion=Free
  swing2LimitDegrees=45
  swing2Motion=Free
  twistLimitDegrees=45
  twistMotion=Free
 }
]
