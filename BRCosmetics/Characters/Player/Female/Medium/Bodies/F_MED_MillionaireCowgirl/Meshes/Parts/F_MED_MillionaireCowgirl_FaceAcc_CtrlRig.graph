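# Post-process style Control Rig for the MillionaireCowgirl face accessory that adjusts dynamic hair and hat bones to follow body pose.
# A sequence constrains a chest control between spine_05 and head, then spherical pose readers on head and right upper arm set weight variables.
# Those variables drive additive ModifyTransforms corrections, plus one constant hat offset.
# Plane-clamp hair collision chains and other constraints exist but are disconnected from execution; crouchmove is unused.
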ControlRigBlueprint /BRCosmetics/Characters/Player/Female/Medium/Bodies/F_MED_MillionaireCowgirl/Meshes/Parts/F_MED_MillionaireCowgirl_FaceAcc_CtrlRig
graphs [
 RigVMFunctionLibrary {links [], nodes []}
 RigVMFunction_Sequence_1_ContainedGraph {nodes [
   n1 = RigVMFunctionEntryNode()
   RigVMFunction_Sequence = &RigVMFunction_Sequence(ExecuteContext=@n1.ExecuteContext)
   n2 = &RigVMFunction_Sequence(ExecuteContext=@RigVMFunction_Sequence.B)
   RigVMFunctionReturnNode(A=@RigVMFunction_Sequence.A, B=@n2.A, C=@n2.B)
  ]}
 RigVMModel {nodes [
   n1 = RigVMRerouteNode(Value=())
   n2 = &RigUnit_GetTransform(Item=(Type=Bone,Name="dyn_hair_bck_c_01"), bInitial=False)
   n3 = &RigUnit_GetTransform(Item=(Type=Bone,Name="dyn_hair_bck_c_02"), bInitial=False)
   ClampSpatially = /Script/RigVM.RigVMFunction_MathTransformClampSpatially(
    Axis=Y
    DebugColor=(R=1,G=0,B=0,A=1)
    DebugThickness=1
    Maximum=200
    Minimum=20
    Type=Plane
    bDrawDebug=true
    Value=@n3.Transform
    Space=&RigUnit_GetTransform(Item=(Type=Control,Name="spine_05_ctrl_c"), bInitial=False).Transform
   ) {
    is=struct RigVMUnitNode
    template="ClampSpatially::Execute(in Value,in Axis,in Type,in Minimum,in Maximum,in Space,in bDrawDebug,in DebugColor,in DebugThickness,out Result)"
   }
   &RigUnit_OffsetTransformForItem(
    Item=(Type=Bone,Name="dyn_hair_bck_c_01")
    OffsetTransform=(Rotation=(X=0,Y=0,Z=0,W=1),Translation=(X=0,Y=0,Z=0),Scale3D=(X=1,Y=1,Z=1))
    Weight=1
    bPropagateToChildren=True
    OffsetTransform.Rotation=/Script/RigVM.RigVMFunction_MathQuaternionFromTwoVectors(
     A=&RigVMFunction_MathTransformMakeRelative(Parent=@n2.Transform, Global=@n3.Transform).Local.Translation
     B=&RigVMFunction_MathTransformMakeRelative(Global=@ClampSpatially.Result, Parent=@n2.Transform).Local.Translation
    ) {is=struct RigVMUnitNode, template="FromTwoVectors::Execute(in A,in B,out Result)"}.Result
    ExecutePin=@n1.Value
   )
   n4 = &RigUnit_GetTransform(Item=(Type=Bone,Name="dyn_hair_bck_l_02"), bInitial=False)
   n5 = &RigVMFunction_MathTransformClampSpatially(
    Axis=Y
    DebugColor=(R=1,G=0,B=0,A=1)
    DebugThickness=1
    Maximum=200
    Minimum=20
    Type=Plane
    bDrawDebug=true
    Space=&RigUnit_GetTransform(Item=(Type=Control,Name="spine_05_ctrl_c"), bInitial=False).Transform
    Value=@n4.Transform
   ) {template="ClampSpatially::Execute(in Value,in Axis,in Type,in Minimum,in Maximum,in Space,in bDrawDebug,in DebugColor,in DebugThickness,out Result)"}
   n6 = &RigUnit_GetTransform(Item=(Type=Bone,Name="dyn_hair_bck_l_01"), bInitial=False)
   > &RigUnit_OffsetTransformForItem(
    Item=(Type=Bone,Name="dyn_hair_bck_l_01")
    OffsetTransform=(Rotation=(X=0,Y=0,Z=0,W=1),Translation=(X=0,Y=0,Z=0),Scale3D=(X=1,Y=1,Z=1))
    Weight=1
    bPropagateToChildren=True
    OffsetTransform.Rotation=&RigVMFunction_MathQuaternionFromTwoVectors(
     B=&RigVMFunction_MathTransformMakeRelative(Global=@n5.Result, Parent=@n6.Transform).Local.Translation
     A=&RigVMFunction_MathTransformMakeRelative(Global=@n4.Transform, Parent=@n6.Transform).Local.Translation
    ) {template="FromTwoVectors::Execute(in A,in B,out Result)"}.Result
   )
   n7 = &RigUnit_GetTransform(Item=(Type=Bone,Name="dyn_hair_bck_r_02"), bInitial=False)
   n8 = &RigVMFunction_MathTransformClampSpatially(
    Axis=Y
    DebugColor=(R=1,G=0,B=0,A=1)
    DebugThickness=1
    Maximum=200
    Minimum=18
    Type=Plane
    bDrawDebug=true
    Space=&RigUnit_GetTransform(Item=(Type=Control,Name="spine_05_ctrl_c"), bInitial=False).Transform
    Value=@n7.Transform
   ) {template="ClampSpatially::Execute(in Value,in Axis,in Type,in Minimum,in Maximum,in Space,in bDrawDebug,in DebugColor,in DebugThickness,out Result)"}
   n9 = &RigUnit_GetTransform(Item=(Type=Bone,Name="dyn_hair_bck_r_01"), bInitial=False)
   > &RigUnit_OffsetTransformForItem(
    Item=(Type=Bone,Name="dyn_hair_bck_r_01")
    OffsetTransform=(Rotation=(X=0,Y=0,Z=0,W=1),Translation=(X=0,Y=0,Z=0),Scale3D=(X=1,Y=1,Z=1))
    Weight=1
    bPropagateToChildren=True
    OffsetTransform.Rotation=&RigVMFunction_MathQuaternionFromTwoVectors(
     B=&RigVMFunction_MathTransformMakeRelative(Global=@n8.Result, Parent=@n9.Transform).Local.Translation
     A=&RigVMFunction_MathTransformMakeRelative(Global=@n7.Transform, Parent=@n9.Transform).Local.Translation
    ) {template="FromTwoVectors::Execute(in A,in B,out Result)"}.Result
   )
   RigUnit_BeginExecution = &RigUnit_BeginExecution()
   RigVMFunction_Sequence = RigVMAggregateNode(ExecuteContext=@RigUnit_BeginExecution.ExecutePin)
   &RigUnit_ParentConstraint(
    AdvancedSettings=(InterpolationType=Average,RotationOrderForFilter=XZY)
    Child=(Type=Control,Name="spine_05_ctrl")
    Parents=((Item=(Type=Bone,Name="spine_05"),Weight=0.75),(Item=(Type=Bone,Name="head"),Weight=0.25))
    Weight=1
    ExecutePin=@RigVMFunction_Sequence.A
   )
   &RigUnit_ModifyTransforms(
    ItemToModify="((Item=(Type=Bone,Name=\"dyn_hair_side_r_01\"),Transform=(Rotation=(X=0,Y=0,Z=0,W=1),Translation=(X=-1,Y=0,Z=-1),Scale3D=(X=1,Y=1,Z=1))),(Item=(Type=Bone,Name=\"dyn_hair_side_r_02\"),Transform=(Rotation=(X=0,Y=0,Z=0,W=1),Translation=(X=-2,Y=0,Z=1),Scale3D=(X=1 ...(truncated)"
    WeightMinimum=0
    ExecutePin=@RigVMFunction_Sequence.C
    Weight=RigVMVariableNode() {variable=R Arm Up}.Value
   )
   n10 = RigVMVariableNode() {variable=look up}
   > &RigUnit_ModifyTransforms(
    ItemToModify="((Item=(Type=Bone,Name=\"dyn_hair_bck_c_01\"),Transform=(Rotation=(X=0,Y=0,Z=0,W=1),Translation=(X=1,Y=0,Z=1),Scale3D=(X=1,Y=1,Z=1))),(Item=(Type=Bone,Name=\"dyn_hair_bck_l_01\"),Transform=(Rotation=(X=0,Y=0,Z=0,W=1),Translation=(X=0,Y=0,Z=1),Scale3D=(X=1,Y=1, ...(truncated)"
    WeightMinimum=0
    Weight=@n10.Value
   )
   > &RigUnit_ModifyTransforms(
    ItemToModify="((Item=(Type=Bone,Name=\"dyn_hair_side_l_01\"),Transform=(Rotation=(X=0,Y=-0.1305,Z=0,W=0.9914),Translation=(X=-1,Y=0,Z=0),Scale3D=(X=1,Y=1,Z=1))),(Item=(Type=Bone,Name=\"dyn_hair_side_r_01\"),Transform=(Rotation=(X=0,Y=-0.1305,Z=0,W=0.9914),Translation=(X=-1, ...(truncated)"
    WeightMinimum=0
    Weight=@n10.Value
   )
   n11 = RigVMVariableNode() {variable=head_back_r}
   > &RigUnit_ModifyTransforms(
    ItemToModify="((Item=(Type=Bone,Name=\"dyn_hair_side_r_01\"),Transform=(Rotation=(X=0,Y=-0.0872,Z=0,W=0.9962),Translation=(X=-2,Y=1,Z=1),Scale3D=(X=1,Y=1,Z=1))),(Item=(Type=Bone,Name=\"dyn_hair_bck_r_01\"),Transform=(Rotation=(X=0,Y=-0.1736,Z=0,W=0.9848),Translation=(X=0,Y= ...(truncated)"
    WeightMinimum=0
    Weight=@n11.Value
   )
   > &RigUnit_ModifyTransforms(
    ItemToModify="((Item=(Type=Bone,Name=\"dyn_hat_bk\"),Transform=(Rotation=(X=0,Y=-0.1736,Z=0,W=0.9848),Translation=(X=0,Y=0,Z=0),Scale3D=(X=1,Y=1,Z=1))),(Item=(Type=Bone,Name=\"dyn_hat_mstr\"),Transform=(Rotation=(X=-0.0076,Y=-0.0868,Z=-0.0868,W=0.9924),Translation=(X=-2,Y=0 ...(truncated)"
    WeightMinimum=0
    Weight=@n11.Value
   )
   > &RigUnit_ModifyTransforms(
    ItemToModify="((Item=(Type=Bone,Name=\"None\"),Transform=(Rotation=(X=0,Y=-0.1736,Z=0,W=0.9848),Translation=(X=0,Y=0,Z=0),Scale3D=(X=1,Y=1,Z=1))),(Item=(Type=Bone,Name=\"dyn_hat_mstr\"),Transform=(Rotation=(X=0,Y=0,Z=-0.0175,W=0.9998),Translation=(X=-0.5,Y=0,Z=0),Scale3D=(X ...(truncated)"
    Weight=1
    WeightMinimum=0
   )
   n12 = &RigUnit_SphericalPoseReader(
    ActiveRegionScaleFactors=(PositiveWidth=2,NegativeWidth=2.5,PositiveHeight=2,NegativeHeight=2.5)
    ActiveRegionSize=0.11
    Debug=(bDrawDebug=false,bDraw2D=false,bDrawLocalAxes=true,DebugScale=25,DebugSegments=20,DebugThickness=0.25)
    DriverAxis=(X=-1,Y=0,Z=0)
    DriverItem=(Type=Bone,Name="upperarm_r")
    FalloffRegionScaleFactors=(PositiveWidth=1,NegativeWidth=1,PositiveHeight=2,NegativeHeight=1)
    FalloffSize=0.36
    FlipHeightScaling=false
    FlipWidthScaling=true
    OptionalParentItem=(Type=Bone,Name="spine_04")
    RotationOffset=(X=-150,Y=50,Z=2.033)
    ExecutePin=@RigVMFunction_Sequence.B
   )
   > RigVMVariableNode(Value=@n12.OutputParam) {variable=R Arm Up}
   > n13 = &RigUnit_SphericalPoseReader(
    ActiveRegionScaleFactors=(PositiveWidth=1,NegativeWidth=0.5,PositiveHeight=1.5,NegativeHeight=1.5)
    ActiveRegionSize=0.1
    Debug=(bDrawDebug=false,DebugScale=25,DebugSegments=20,DebugThickness=0.25)
    DriverAxis=(X=0,Y=1,Z=0)
    DriverItem=(Type=Bone,Name="head")
    FalloffRegionScaleFactors=(PositiveWidth=1,NegativeWidth=0.5,PositiveHeight=1.5,NegativeHeight=1.5)
    FalloffSize=0.3
    FlipWidthScaling=false
    OptionalParentItem=(Type=Bone)
    RotationOffset=(X=90,Y=0,Z=-30)
   )
   > RigVMVariableNode(Value=@n13.OutputParam) {variable=look up}
   > n14 = &RigUnit_SphericalPoseReader(
    ActiveRegionScaleFactors=(PositiveWidth=1.6,NegativeWidth=3,PositiveHeight=1,NegativeHeight=3)
    ActiveRegionSize=0.1
    Debug=(bDrawDebug=true,DebugScale=25,DebugSegments=20,DebugThickness=0.25)
    DriverAxis=(X=1,Y=0,Z=0)
    DriverItem=(Type=Bone,Name="head")
    FalloffRegionScaleFactors=(PositiveWidth=0.35,NegativeWidth=0.5,PositiveHeight=0.35,NegativeHeight=0.5)
    FalloffSize=0.5
    FlipHeightScaling=true
    OptionalParentItem=(Type=Bone,Name="spine_05")
    RotationOffset=(X=-35,Y=0,Z=0)
   )
   > RigVMVariableNode(Value=@n14.OutputParam) {variable=head_back_r}
   &RigUnit_ModifyTransforms(
    ExecutePin=()
    ItemToModify="((Item=(Type=Bone,Name=\"dyn_hair_side_r_01\"),Transform=(Rotation=(X=0,Y=0,Z=0,W=1),Translation=(X=0,Y=0,Z=0),Scale3D=(X=1,Y=1,Z=1))),(Item=(Type=Bone,Name=\"dyn_hair_side_r_02\"),Transform=(Rotation=(X=0,Y=0,Z=0,W=1),Translation=(X=0,Y=0,Z=0),Scale3D=(X=1,Y= ...(truncated)"
    Weight=1
    WeightMinimum=0
   )
   n15 = &RigUnit_GetTransform(Item=(Type=Bone,Name="dyn_hair_side_r_02"), bInitial=False)
   n16 = &RigVMFunction_MathTransformClampSpatially(
    Axis=X
    DebugColor=(R=1,G=0,B=0,A=1)
    DebugThickness=1
    Maximum=200
    Minimum=15
    Type=Plane
    bDrawDebug=true
    Space=&RigUnit_GetTransform(Item=(Type=Control,Name="spine_05_ctrl_c_2"), bInitial=False).Transform
    Value=@n15.Transform
   ) {template="ClampSpatially::Execute(in Value,in Axis,in Type,in Minimum,in Maximum,in Space,in bDrawDebug,in DebugColor,in DebugThickness,out Result)"}
   n17 = &RigUnit_GetTransform(Item=(Type=Bone,Name="dyn_hair_side_r_01"), bInitial=False)
   &RigUnit_OffsetTransformForItem(
    ExecutePin=()
    Item=(Type=Bone,Name="dyn_hair_side_r_01")
    OffsetTransform=(Rotation=(X=0,Y=0,Z=0,W=1),Translation=(X=0,Y=0,Z=0),Scale3D=(X=1,Y=1,Z=1))
    Weight=1
    bPropagateToChildren=True
    OffsetTransform.Rotation=&RigVMFunction_MathQuaternionFromTwoVectors(
     B=&RigVMFunction_MathTransformMakeRelative(Global=@n16.Result, Parent=@n17.Transform).Local.Translation
     A=&RigVMFunction_MathTransformMakeRelative(Global=@n15.Transform, Parent=@n17.Transform).Local.Translation
    ) {template="FromTwoVectors::Execute(in A,in B,out Result)"}.Result
   )
   &RigUnit_ParentConstraint(
    AdvancedSettings=(InterpolationType=Average,RotationOrderForFilter=XZY)
    Child=(Type=Bone,Name="dyn_hair_side_r_02")
    ExecutePin=()
    Parents=((Item=(Type=Bone,Name="dyn_hair_bck_r_02"),Weight=1))
    Weight=1
   )
   &RigUnit_ParentConstraint(
    AdvancedSettings=(InterpolationType=Average,RotationOrderForFilter=XZY)
    Child=(Type=Bone,Name="dyn_hair_fr_l_02")
    ExecutePin=()
    Parents=((Item=(Type=Bone,Name="dyn_hair_side_l_02"),Weight=1))
    Weight=1
   )
  ]}
]
hierarchy [
 root {loc=<0 0 0>, rot=<0 0 0 1>}
 pelvis {parent=root, loc=<0 -0.5301 98.0571>, rot=<0.0127 0.707 -0.0127 -0.707>}
 spine_01 {parent=pelvis, loc=<6.1208 -0.9328 0>, rot=<0 0 0.034 -0.9994>}
 spine_02 {parent=spine_01, loc=<6.5996 -0.2095 0>, rot=<0 0 0.0238 -0.9997>}
 spine_03 {parent=spine_02, loc=<6.4914 -0.1237 0>, rot=<0 0 0.0149 -0.9999>}
 spine_04 {parent=spine_03, loc=<6.9598 0.0794 0>, rot=<0 0 -0.0132 -0.9999>}
 spine_05 {parent=spine_04, loc=<6.9987 0 0>, rot=<0 0 -0.0015 -1>}
 neck_01 {parent=spine_05, loc=<11.3138 0.1529 0>, rot=<0 0 -0.2003 -0.9797>}
 neck_02 {parent=neck_01, loc=<4.5627 0 0>, rot=<0 0 0.0504 -0.9987>}
 head {parent=neck_02, loc=<5.3914 0 0>, rot=<0 0 0.115 -0.9934>}
 dyn_hat_mstr {parent=head, loc=<6.8393 0.6143 0>, rot=<0.0029 -0.7071 0.7071 -0.0029>}
 dyn_hat_bk {parent=dyn_hat_mstr, loc=<2.4878 0 9.2902>, rot=<0 0 0 1>}
 dyn_hair_fr_l_MSTR {parent=head, loc=<1.3141 3.2149 -8.0116>, rot=<0.1582 -0.9874 0 0>}
 dyn_hair_fr_l_01 {parent=dyn_hair_fr_l_MSTR, loc=<0 0 0>, rot=<0 0 0 1>}
 dyn_hair_fr_l_02 {parent=dyn_hair_fr_l_01, loc=<5.8249 0 0>, rot=<0 0 -0.0238 -0.9997>}
 dyn_hair_side_l_MSTR {parent=head, loc=<1.5971 -2.4153 -7.3149>, rot=<0.2432 -0.9648 0.0238 -0.0977>}
 dyn_hair_side_l_01 {parent=dyn_hair_side_l_MSTR, loc=<0 0 0>, rot=<0 0 0 1>}
 dyn_hair_side_l_02 {parent=dyn_hair_side_l_01, loc=<7.1713 0 0>, rot=<-0.0791 0.0653 0.0161 -0.9946>}
 dyn_hair_bck_l_MSTR {parent=head, loc=<-0.1039 -7.5073 -3.8017>, rot=<0.0633 -0.7828 0.5787 -0.2197>}
 dyn_hair_bck_l_01 {parent=dyn_hair_bck_l_MSTR, loc=<0 0 0>, rot=<0 0 0 1>}
 dyn_hair_bck_l_02 {parent=dyn_hair_bck_l_01, loc=<7.9405 0 0>, rot=<0.0245 0.0342 0.0241 -0.9988>}
 dyn_hair_bck_c_MSTR {parent=head, loc=<-1.5832 -7.6026 0>, rot=<0.2061 -0.6764 0.6764 -0.2061>}
 dyn_hair_bck_c_01 {parent=dyn_hair_bck_c_MSTR, loc=<0 0 0>, rot=<0 0 0 1>}
 dyn_hair_bck_c_02 {parent=dyn_hair_bck_c_01, loc=<8.4922 0 0>, rot=<0 -0.0299 0 -0.9996>}
 dyn_hair_bck_r_MSTR {parent=head, loc=<-0.1041 -7.4804 4.3596>, rot=<0.1607 -0.3904 0.8965 -0.1341>}
 dyn_hair_bck_r_01 {parent=dyn_hair_bck_r_MSTR, loc=<0 0 0>, rot=<0 0 0 1>}
 dyn_hair_bck_r_02 {parent=dyn_hair_bck_r_01, loc=<7.3193 0 0>, rot=<0.1421 0.0486 -0.0155 -0.9885>}
 dyn_hair_side_r_MSTR {parent=head, loc=<0.9139 -1.1368 7.3647>, rot=<0.1058 -0.0975 0.9824 -0.1194>}
 dyn_hair_side_r_01 {parent=dyn_hair_side_r_MSTR, loc=<0 0 0>, rot=<0 0 0 1>}
 dyn_hair_side_r_02 {parent=dyn_hair_side_r_01, loc=<7.1127 0 0>, rot=<0.0805 0.1033 -0.0331 -0.9908>}
 dyn_hair_fr_r_MSTR {parent=head, loc=<1.2937 5.6841 7.0182>, rot=<0.0056 -0.0433 0.9909 -0.1275>}
 dyn_hair_fr_r_01 {parent=dyn_hair_fr_r_MSTR, loc=<0 0 0>, rot=<0 0 0 1>}
 dyn_hair_fr_r_02 {parent=dyn_hair_fr_r_01, loc=<5.6851 0 0>, rot=<0.2703 0.0571 -0.0307 -0.9606>}
 clavicle_l {parent=spine_05, loc=<5.4831 -0.1053 -3.2946>, rot=<0.0465 -0.7289 -0.009 -0.6829>}
 upperarm_l {parent=clavicle_l, loc=<9.8658 -1.2136 -0.4105>, rot=<0.0164 -0.3245 -0.0269 -0.9454>}
 clavicle_r {parent=spine_05, loc=<5.4831 -0.1053 3.2946>, rot=<0.7289 0.0465 -0.6829 0.009>}
 upperarm_r {parent=clavicle_r, loc=<-9.8658 1.2136 0.4105>, rot=<0.0164 -0.3245 -0.0269 -0.9454>}
 spine_05_ctrl {controlType=EULER_TRANSFORM, parent=spine_05, type=Control, loc=<0 0 0>, rot=<0 0 0 1>}
 spine_05_ctrl_c {controlType=EULER_TRANSFORM, parent=spine_05_ctrl, type=Control, loc=<0 0 0>, rot=<0 0 0 1>}
 spine_05_ctrl_c_2 {controlType=EULER_TRANSFORM, parent=spine_05_ctrl, type=Control, loc=<0 0 0>, rot=<0 0 0 1>}
]
variables [
 R Arm Up {type=double, default=0}
 crouchmove {type=double, default=0}
 head_back_r {type=double, default=0}
 look up {type=double, default=0}
]
