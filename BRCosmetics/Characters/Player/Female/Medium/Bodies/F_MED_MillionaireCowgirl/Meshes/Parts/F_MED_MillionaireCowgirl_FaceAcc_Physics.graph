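# Physics asset for the MillionaireCowgirl medium female body's FaceAcc mesh, providing capsule bodies for dynamic hair.
# Seven hair chains (three back, two front, two side) each run from a small master body on the head through two segments with tuned swing limits.
# Head, neck, upper spine, clavicle and upper-arm capsules approximate the upper body around the hair.
# An extra free-angular constraint ties the left front tip to the left side tip.
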PhysicsAsset /BRCosmetics/Characters/Player/Female/Medium/Bodies/F_MED_MillionaireCowgirl/Meshes/Parts/F_MED_MillionaireCowgirl_FaceAcc_Physics
bodies {
 clavicle_l [{shapeType=Capsule, length=5, radius=6, shapeName=clavicle_l_capsule, center=<5.7154 -1.2368 -0.3174>, rotation=<rot -80 179.9998 -179.9998>}]
 clavicle_r [{shapeType=Capsule, length=8, radius=6.2, shapeName=clavicle_r_capsule, center=<-7.0578 1.1088 0.2609>, rotation=<rot 85 -0.0006 -0.0006>}]
 dyn_hair_bck_c_01 [{shapeType=Capsule, length=2.0008, radius=4.1242, shapeName=dyn_hair_bck_c_01_capsule, center=<3 0 2>, rotation=<rot 90 0 0>}]
 dyn_hair_bck_c_02 [{shapeType=Capsule, length=1.5, radius=4.3351, shapeName=dyn_hair_bck_c_02_capsule, center=<3 0 1.5047>, rotation=<rot 90 0 0>}]
 dyn_hair_bck_c_MSTR [{shapeType=Capsule, length=0.505, radius=0.505, shapeName=dyn_hair_bck_c_MSTR_capsule, rotation=<rot 0 -90 0>}]
 dyn_hair_bck_l_01 [{shapeType=Capsule, length=2.0008, radius=4.1242, shapeName=dyn_hair_bck_l_01_capsule, center=<3 0 2>, rotation=<rot 90 0 0>}]
 dyn_hair_bck_l_02 [{shapeType=Capsule, length=2.1258, radius=3.1242, shapeName=dyn_hair_bck_l_02_capsule, center=<3 0 1.5059>, rotation=<rot 90 0 0>}]
 dyn_hair_bck_l_MSTR [{shapeType=Capsule, length=0.505, radius=0.505, shapeName=dyn_hair_bck_l_MSTR_capsule, rotation=<rot 0 -90 0>}]
 dyn_hair_bck_r_01 [{shapeType=Capsule, length=1.5919, radius=4.5331, shapeName=dyn_hair_bck_r_01_capsule, center=<3 0 2>, rotation=<rot 90 0 0>}]
 dyn_hair_bck_r_02 [{shapeType=Capsule, length=1.8062, radius=3.9974, shapeName=dyn_hair_bck_r_02_capsule, center=<3 0 1.3415>, rotation=<rot 90 0 0>}]
 dyn_hair_bck_r_MSTR [{shapeType=Capsule, length=0.505, radius=0.505, shapeName=dyn_hair_bck_r_MSTR_capsule, rotation=<rot 0 -90 0>}]
 dyn_hair_fr_l_01 [{shapeType=Capsule, length=1.5, radius=2.5, shapeName=dyn_hair_fr_l_01_capsule, center=<1.8175 0 0.3205>, rotation=<rot 80 -179.9998 -180>}]
 dyn_hair_fr_l_02 [{shapeType=Capsule, length=1, radius=4, shapeName=dyn_hair_fr_l_02_capsule, center=<1.3803 0 0.2434>, rotation=<rot 80 -179.9997 -180>}]
 dyn_hair_fr_l_MSTR [{shapeType=Capsule, length=0.505, radius=0.505, shapeName=dyn_hair_fr_l_MSTR_capsule, rotation=<rot 0 -90 0>}]
 dyn_hair_fr_r_01 [{shapeType=Capsule, length=2, radius=2, shapeName=dyn_hair_fr_r_01_capsule, center=<1.8175 0 0.3205>, rotation=<rot 80 179.9998 -180>}]
 dyn_hair_fr_r_02 [{
   shapeType=Capsule
   length=0
   radius=3
   shapeName=dyn_hair_fr_r_02_capsule
   center=<2.4188 0 0.4265>
   rotation=<rot 80.0001 179.9997 179.9999>
  }]
 dyn_hair_fr_r_MSTR [{shapeType=Capsule, length=0.505, radius=0.505, shapeName=dyn_hair_fr_r_MSTR_capsule, rotation=<rot 0 -90 0>}]
 dyn_hair_side_l_01 [{
   shapeType=Capsule
   length=1.5
   radius=3
   shapeName=dyn_hair_side_l_01_capsule
   center=<1.8175 0 0.3205>
   rotation=<rot 80 -179.9996 -179.9999>
  }]
 dyn_hair_side_l_02 [{
   shapeType=Capsule
   length=4.25
   radius=2.75
   shapeName=dyn_hair_side_l_02_capsule
   center=<1.8307 0.1531 0.3228>
   rotation=<rot 80.0001 -179.9994 -174.9997>
  }]
 dyn_hair_side_l_MSTR [{shapeType=Capsule, length=0.505, radius=0.505, shapeName=dyn_hair_side_l_MSTR_capsule, rotation=<rot 0 -90 0>}]
 dyn_hair_side_r_01 [{shapeType=Capsule, length=2, radius=2.5, shapeName=dyn_hair_side_r_01_capsule, center=<1.8175 0 0.3205>, rotation=<rot 80 179.9996 179.9999>}]
 dyn_hair_side_r_02 [{
   shapeType=Capsule
   length=5.4797
   radius=2.3138
   shapeName=dyn_hair_side_r_02_capsule
   center=<2.841 1.7404 0.6738>
   rotation=<rot 78.8801 -17.9355 -57.3392>
  }]
 dyn_hair_side_r_MSTR [{shapeType=Capsule, length=0.505, radius=0.505, shapeName=dyn_hair_side_r_MSTR_capsule, rotation=<rot 0 -90 0>}]
 head [{shapeType=Capsule, length=0, radius=4, shapeName=head_capsule}]
 neck_01 [{shapeType=Capsule, length=3.5, radius=4, shapeName=neck_01_capsule, rotation=<rot 0 -90.0002 -90>}]
 neck_02 [{shapeType=Capsule, length=3.5, radius=4, shapeName=neck_02_capsule, rotation=<rot 0 -90.0002 -90>}]
 spine_05 [
  {shapeType=Capsule, length=9, radius=9, shapeName=spine_05_capsule, center=<3.9385 -1.5053 0>, rotation=<rot 0 -90 0>}
  {shapeType=Capsule, length=5.116, radius=8.5, shapeName=spine_05_capsule, center=<9.6059 4.6613 0>, rotation=<rot 0 -90 0>}
 ]
 upperarm_l [{shapeType=Capsule, length=8, radius=5, shapeName=upperarm_l_capsule, center=<8.2173 0 0.3398>, rotation=<rot -0.0003 -89.9998 90>}]
 upperarm_r [{shapeType=Capsule, length=8, radius=5, shapeName=upperarm_r_capsule, center=<-8.2173 -0.0001 -0.3398>, rotation=<rot -0.0005 -89.9996 -90>}]
}
constraints [
 {bone1Name=clavicle_l, bone2Name=spine_05, swing1LimitDegrees=45, swing2LimitDegrees=45, twistLimitDegrees=0}
 {
  bone1Name=upperarm_l
  bone2Name=clavicle_l
  swing1LimitDegrees=45
  swing1Motion=Free
  swing2LimitDegrees=45
  swing2Motion=Free
  twistLimitDegrees=0
  twistMotion=Free
 }
 {bone1Name=clavicle_r, bone2Name=spine_05, swing1LimitDegrees=45, swing2LimitDegrees=45, twistLimitDegrees=0}
 {
  bone1Name=upperarm_r
  bone2Name=clavicle_r
  swing1LimitDegrees=45
  swing1Motion=Free
  swing2LimitDegrees=45
  swing2Motion=Free
  twistLimitDegrees=0
  twistMotion=Free
 }
 {bone1Name=neck_01, bone2Name=spine_05, swing1LimitDegrees=45, swing2LimitDegrees=45, twistLimitDegrees=0}
 {bone1Name=neck_02, bone2Name=neck_01, swing1LimitDegrees=45, swing2LimitDegrees=45, twistLimitDegrees=0}
 {bone1Name=head, bone2Name=neck_02, swing1LimitDegrees=45, swing2LimitDegrees=45, twistLimitDegrees=0}
 {
  bone1Name=dyn_hair_bck_c_MSTR
  bone2Name=head
  swing1LimitDegrees=45
  swing1Motion=Free
  swing2LimitDegrees=45
  swing2Motion=Free
  twistLimitDegrees=45
  twistMotion=Free
 }
 {bone1Name=dyn_hair_bck_c_01, bone2Name=dyn_hair_bck_c_MSTR, swing1LimitDegrees=15, swing2LimitDegrees=45, twistLimitDegrees=0}
 {bone1Name=dyn_hair_bck_c_02, bone2Name=dyn_hair_bck_c_01, swing1LimitDegrees=10, swing2LimitDegrees=50, twistLimitDegrees=0}
 {
  bone1Name=dyn_hair_bck_l_MSTR
  bone2Name=head
  swing1LimitDegrees=45
  swing1Motion=Free
  swing2LimitDegrees=45
  swing2Motion=Free
  twistLimitDegrees=45
  twistMotion=Free
 }
 {bone1Name=dyn_hair_bck_l_01, bone2Name=dyn_hair_bck_l_MSTR, swing1LimitDegrees=15, swing2LimitDegrees=45, twistLimitDegrees=0}
 {bone1Name=dyn_hair_bck_l_02, bone2Name=dyn_hair_bck_l_01, swing1LimitDegrees=15, swing2LimitDegrees=45, twistLimitDegrees=0}
 {
  bone1Name=dyn_hair_bck_r_MSTR
  bone2Name=head
  swing1LimitDegrees=45
  swing1Motion=Free
  swing2LimitDegrees=45
  swing2Motion=Free
  twistLimitDegrees=45
  twistMotion=Free
 }
 {bone1Name=dyn_hair_bck_r_01, bone2Name=dyn_hair_bck_r_MSTR, swing1LimitDegrees=20, swing2LimitDegrees=45, twistLimitDegrees=0}
 {bone1Name=dyn_hair_bck_r_02, bone2Name=dyn_hair_bck_r_01, swing1LimitDegrees=20, swing2LimitDegrees=60, twistLimitDegrees=0}
 {
  bone1Name=dyn_hair_fr_l_MSTR
  bone2Name=head
  swing1LimitDegrees=45
  swing1Motion=Free
  swing2LimitDegrees=45
  swing2Motion=Free
  twistLimitDegrees=45
  twistMotion=Free
 }
 {bone1Name=dyn_hair_fr_l_01, bone2Name=dyn_hair_fr_l_MSTR, swing1LimitDegrees=15, swing2LimitDegrees=45, twistLimitDegrees=0}
 {bone1Name=dyn_hair_fr_l_02, bone2Name=dyn_hair_fr_l_01, swing1LimitDegrees=15, swing2LimitDegrees=30, twistLimitDegrees=0}
 {
  bone1Name=dyn_hair_fr_l_02
  bone2Name=dyn_hair_side_l_02
  swing1LimitDegrees=45
  swing1Motion=Free
  swing2LimitDegrees=45
  swing2Motion=Free
  twistLimitDegrees=45
  twistMotion=Free
 }
 {
  bone1Name=dyn_hair_fr_r_MSTR
  bone2Name=head
  swing1LimitDegrees=45
  swing1Motion=Free
  swing2LimitDegrees=45
  swing2Motion=Free
  twistLimitDegrees=45
  twistMotion=Free
 }
 {bone1Name=dyn_hair_fr_r_01, bone2Name=dyn_hair_fr_r_MSTR, swing1LimitDegrees=15, swing2LimitDegrees=45, twistLimitDegrees=0}
 {bone1Name=dyn_hair_fr_r_02, bone2Name=dyn_hair_fr_r_01, swing1LimitDegrees=15, swing2LimitDegrees=30, twistLimitDegrees=0}
 {
  bone1Name=dyn_hair_side_l_MSTR
  bone2Name=head
  swing1LimitDegrees=45
  swing1Motion=Free
  swing2LimitDegrees=45
  swing2Motion=Free
  twistLimitDegrees=45
  twistMotion=Free
 }
 {bone1Name=dyn_hair_side_l_01, bone2Name=dyn_hair_side_l_MSTR, swing1LimitDegrees=15, swing2LimitDegrees=45, twistLimitDegrees=0}
 {bone1Name=dyn_hair_side_l_02, bone2Name=dyn_hair_side_l_01, swing1LimitDegrees=15, swing2LimitDegrees=45, twistLimitDegrees=0}
 {
  bone1Name=dyn_hair_side_r_MSTR
  bone2Name=head
  swing1LimitDegrees=45
  swing1Motion=Free
  swing2LimitDegrees=45
  swing2Motion=Free
  twistLimitDegrees=45
  twistMotion=Free
 }
 {bone1Name=dyn_hair_side_r_01, bone2Name=dyn_hair_side_r_MSTR, swing1LimitDegrees=30, swing2LimitDegrees=45, twistLimitDegrees=0}
 {bone1Name=dyn_hair_side_r_02, bone2Name=dyn_hair_side_r_01, swing1LimitDegrees=35, swing2LimitDegrees=50, twistLimitDegrees=0}
]
