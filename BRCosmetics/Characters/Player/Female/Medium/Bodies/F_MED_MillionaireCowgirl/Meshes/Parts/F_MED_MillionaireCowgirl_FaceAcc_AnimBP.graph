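# Anim Blueprint for the MillionaireCowgirl face accessory, driving secondary motion of hair and hat bones.
# It blends copied mesh poses at the upper arms, spreads hair roots with weighted constraints, and applies additive corrections, some scaled by head pitch or enabled in the front end.
# RigidBody simulation in spine_05 space follows, then right-side cleanup and a Control Rig produce the output.
# An Anim Dynamics hair chain remains unconnected to the output.
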AnimBlueprint /BRCosmetics/Characters/Player/Female/Medium/Bodies/F_MED_MillionaireCowgirl/Meshes/Parts/F_MED_MillionaireCowgirl_FaceAcc_AnimBP
target_skeleton=&/F_MED_MillionaireCowgirl_FaceAcc_Skeleton
graphs [
 AnimGraph {nodes [
   n1 = Anim Dynamics - Bone: dyn_hair_fr_l_01(Alpha=1, ExternalForce="0,0,0") {
    angularDampingOverride=0.8
    angularSpringConstant=5000
    bAngularSpring=true
    bGravityOverrideInSimSpace=true
    bUseGravityOverride=true
    alphaBoolBlend {blendOption=Linear}
    boundBone {boneName=dyn_hair_fr_l_01}
    componentAppliedLinearAccClamp=<10000 10000 10000>
    componentLinearAccScale=<1 1 1>
    componentLinearVelScale=<1 1 1>
    gravityOverride {z=-1200}
    physicsBodyDefinitions [{
      boundBone {boneName=dyn_hair_fr_l_01}
      boxExtents=<9 9 9>
      constraintSetup {angularLimitsMax=<5 20 5>, angularLimitsMin=<0 -50 -20>, angularTarget=<0 0 0>}
      localJointOffset=<0.1 0 0>
     }]
    relativeSpaceBone {boneName=neck_01}
   }
   n2 = Anim Dynamics - Bone: dyn_hair_fr_l_02(ComponentPose=@n1, Alpha=1, ExternalForce="0,0,0") {
    angularDampingOverride=0.8
    angularSpringConstant=5000
    bAngularSpring=true
    bGravityOverrideInSimSpace=true
    bUseGravityOverride=true
    alphaBoolBlend {blendOption=Linear}
    boundBone {boneName=dyn_hair_fr_l_02}
    componentAppliedLinearAccClamp=<10000 10000 10000>
    componentLinearAccScale=<1 1 1>
    componentLinearVelScale=<1 1 1>
    gravityOverride {z=-1200}
    physicsBodyDefinitions [{
      boundBone {boneName=dyn_hair_fr_l_02}
      boxExtents=<9 9 9>
      constraintSetup {angularLimitsMax=<5 20 5>, angularLimitsMin=<0 -70 -20>, angularTarget=<0 0 0>}
      localJointOffset=<0.1 0 0>
     }]
    relativeSpaceBone {boneName=neck_01}
   }
   n3 = Anim Dynamics - Bone: dyn_hair_side_l_01(ComponentPose=@n2, Alpha=1, ExternalForce="0,0,0") {
    angularDampingOverride=0.8
    angularSpringConstant=5000
    bAngularSpring=true
    bGravityOverrideInSimSpace=true
    bUseGravityOverride=true
    alphaBoolBlend {blendOption=Linear}
    boundBone {boneName=dyn_hair_side_l_01}
    componentAppliedLinearAccClamp=<10000 10000 10000>
    componentLinearAccScale=<1 1 1>
    componentLinearVelScale=<1 1 1>
    gravityOverride {z=-1200}
    physicsBodyDefinitions [{
      boundBone {boneName=dyn_hair_side_l_01}
      boxExtents=<9 9 9>
      constraintSetup {angularLimitsMax=<5 20 5>, angularLimitsMin=<0 -50 -20>, angularTarget=<0 0 0>}
      localJointOffset=<0.1 0 0>
     }]
    relativeSpaceBone {boneName=neck_01}
   }
   n4 = Anim Dynamics - Bone: dyn_hair_side_l_02(ComponentPose=@n3, Alpha=1, ExternalForce="0,0,0") {
    angularDampingOverride=0.8
    angularSpringConstant=5000
    bAngularSpring=true
    bGravityOverrideInSimSpace=true
    bUseGravityOverride=true
    alphaBoolBlend {blendOption=Linear}
    boundBone {boneName=dyn_hair_side_l_02}
    componentAppliedLinearAccClamp=<10000 10000 10000>
    componentLinearAccScale=<1 1 1>
    componentLinearVelScale=<1 1 1>
    gravityOverride {z=-1200}
    physicsBodyDefinitions [{
      boundBone {boneName=dyn_hair_side_l_02}
      boxExtents=<9 9 9>
      constraintSetup {angularLimitsMax=<5 20 5>, angularLimitsMin=<0 -70 -20>, angularTarget=<0 0 0>}
      localJointOffset=<0.1 0 0>
     }]
    relativeSpaceBone {boneName=neck_01}
   }
   n5 = Anim Dynamics - Bone: dyn_hair_bck_l_01(ComponentPose=@n4, Alpha=1, ExternalForce="0,0,0") {
    angularDampingOverride=0.8
    angularSpringConstant=5000
    bAngularSpring=true
    bGravityOverrideInSimSpace=true
    bUseGravityOverride=true
    alphaBoolBlend {blendOption=Linear}
    boundBone {boneName=dyn_hair_bck_l_01}
    componentAppliedLinearAccClamp=<10000 10000 10000>
    componentLinearAccScale=<1 1 1>
    componentLinearVelScale=<1 1 1>
    gravityOverride {z=-1200}
    physicsBodyDefinitions [{
      boundBone {boneName=dyn_hair_bck_l_01}
      boxExtents=<9 9 9>
      constraintSetup {angularLimitsMax=<5 20 5>, angularLimitsMin=<0 -50 -20>, angularTarget=<0 0 0>}
      localJointOffset=<0.1 0 0>
     }]
    relativeSpaceBone {boneName=neck_01}
   }
   n6 = Anim Dynamics - Bone: dyn_hair_bck_l_02(ComponentPose=@n5, Alpha=1, ExternalForce="0,0,0") {
    angularDampingOverride=0.8
    angularSpringConstant=5000
    bAngularSpring=true
    bGravityOverrideInSimSpace=true
    bUseGravityOverride=true
    alphaBoolBlend {blendOption=Linear}
    boundBone {boneName=dyn_hair_bck_l_02}
    componentAppliedLinearAccClamp=<10000 10000 10000>
    componentLinearAccScale=<1 1 1>
    componentLinearVelScale=<1 1 1>
    gravityOverride {z=-1200}
    physicsBodyDefinitions [{
      boundBone {boneName=dyn_hair_bck_l_02}
      boxExtents=<9 9 9>
      constraintSetup {angularLimitsMax=<5 20 5>, angularLimitsMin=<0 -70 -20>, angularTarget=<0 0 0>}
      localJointOffset=<0.1 0 0>
     }]
    relativeSpaceBone {boneName=neck_01}
   }
   n7 = Anim Dynamics - Bone: dyn_hair_bck_c_01(ComponentPose=@n6, Alpha=1, ExternalForce="0,0,0") {
    angularDampingOverride=0.8
    angularSpringConstant=5000
    bAngularSpring=true
    bGravityOverrideInSimSpace=true
    bUseGravityOverride=true
    alphaBoolBlend {blendOption=Linear}
    boundBone {boneName=dyn_hair_bck_c_01}
    componentAppliedLinearAccClamp=<10000 10000 10000>
    componentLinearAccScale=<1 1 1>
    componentLinearVelScale=<1 1 1>
    gravityOverride {z=-1200}
    physicsBodyDefinitions [{
      boundBone {boneName=dyn_hair_bck_c_01}
      boxExtents=<9 9 9>
      constraintSetup {angularLimitsMax=<5 20 5>, angularLimitsMin=<0 -50 -20>, angularTarget=<0 0 0>}
      localJointOffset=<0.1 0 0>
     }]
    relativeSpaceBone {boneName=neck_01}
   }
   n8 = Anim Dynamics - Bone: dyn_hair_bck_c_02(ComponentPose=@n7, Alpha=1, ExternalForce="0,0,0") {
    angularDampingOverride=0.8
    angularSpringConstant=5000
    bAngularSpring=true
    bGravityOverrideInSimSpace=true
    bUseGravityOverride=true
    alphaBoolBlend {blendOption=Linear}
    boundBone {boneName=dyn_hair_bck_c_02}
    componentAppliedLinearAccClamp=<10000 10000 10000>
    componentLinearAccScale=<1 1 1>
    componentLinearVelScale=<1 1 1>
    gravityOverride {z=-1200}
    physicsBodyDefinitions [{
      boundBone {boneName=dyn_hair_bck_c_02}
      boxExtents=<9 9 9>
      constraintSetup {angularLimitsMax=<5 20 5>, angularLimitsMin=<0 -70 -20>, angularTarget=<0 0 0>}
      localJointOffset=<0.1 0 0>
     }]
    relativeSpaceBone {boneName=neck_01}
   }
   n9 = Anim Dynamics - Bone: dyn_hair_bck_r_01(ComponentPose=@n8, Alpha=1, ExternalForce="0,0,0") {
    angularDampingOverride=0.8
    angularSpringConstant=5000
    bAngularSpring=true
    bGravityOverrideInSimSpace=true
    bUseGravityOverride=true
    alphaBoolBlend {blendOption=Linear}
    boundBone {boneName=dyn_hair_bck_r_01}
    componentAppliedLinearAccClamp=<10000 10000 10000>
    componentLinearAccScale=<1 1 1>
    componentLinearVelScale=<1 1 1>
    gravityOverride {z=-1200}
    physicsBodyDefinitions [{
      boundBone {boneName=dyn_hair_bck_r_01}
      boxExtents=<9 9 9>
      constraintSetup {angularLimitsMax=<5 20 5>, angularLimitsMin=<0 -50 -20>, angularTarget=<0 0 0>}
      localJointOffset=<0.1 0 0>
     }]
    relativeSpaceBone {boneName=neck_01}
   }
   n10 = Anim Dynamics - Bone: dyn_hair_bck_r_02(ComponentPose=@n9, Alpha=1, ExternalForce="0,0,0") {
    angularDampingOverride=0.8
    angularSpringConstant=5000
    bAngularSpring=true
    bGravityOverrideInSimSpace=true
    bUseGravityOverride=true
    alphaBoolBlend {blendOption=Linear}
    boundBone {boneName=dyn_hair_bck_r_02}
    componentAppliedLinearAccClamp=<10000 10000 10000>
    componentLinearAccScale=<1 1 1>
    componentLinearVelScale=<1 1 1>
    gravityOverride {z=-1200}
    physicsBodyDefinitions [{
      boundBone {boneName=dyn_hair_bck_r_02}
      boxExtents=<9 9 9>
      constraintSetup {angularLimitsMax=<5 20 5>, angularLimitsMin=<0 -70 -20>, angularTarget=<0 0 0>}
      localJointOffset=<0.1 0 0>
     }]
    relativeSpaceBone {boneName=neck_01}
   }
   n11 = Anim Dynamics - Bone: dyn_hair_side_r_01(ComponentPose=@n10, Alpha=1, ExternalForce="0,0,0") {
    angularDampingOverride=0.8
    angularSpringConstant=5000
    bAngularSpring=true
    bGravityOverrideInSimSpace=true
    bUseGravityOverride=true
    alphaBoolBlend {blendOption=Linear}
    boundBone {boneName=dyn_hair_side_r_01}
    componentAppliedLinearAccClamp=<10000 10000 10000>
    componentLinearAccScale=<1 1 1>
    componentLinearVelScale=<1 1 1>
    gravityOverride {z=-1200}
    physicsBodyDefinitions [{
      boundBone {boneName=dyn_hair_side_r_01}
      boxExtents=<9 9 9>
      constraintSetup {angularLimitsMax=<5 2 5>, angularLimitsMin=<0 -50 -20>, angularTarget=<0 0 0>}
      localJointOffset=<0.1 0 0>
     }]
    relativeSpaceBone {boneName=neck_01}
   }
   n12 = Anim Dynamics - Bone: dyn_hair_side_r_02(ComponentPose=@n11, Alpha=1, ExternalForce="0,0,0") {
    angularDampingOverride=0.8
    angularSpringConstant=5000
    bAngularSpring=true
    bGravityOverrideInSimSpace=true
    bUseGravityOverride=true
    alphaBoolBlend {blendOption=Linear}
    boundBone {boneName=dyn_hair_side_r_02}
    componentAppliedLinearAccClamp=<10000 10000 10000>
    componentLinearAccScale=<1 1 1>
    componentLinearVelScale=<1 1 1>
    gravityOverride {z=-1200}
    physicsBodyDefinitions [{
      boundBone {boneName=dyn_hair_side_r_02}
      boxExtents=<9 9 9>
      constraintSetup {angularLimitsMax=<5 2 5>, angularLimitsMin=<0 -70 -20>, angularTarget=<0 0 0>}
      localJointOffset=<0.1 0 0>
     }]
    relativeSpaceBone {boneName=neck_01}
   }
   n13 = Anim Dynamics - Bone: dyn_hair_fr_r_01(ComponentPose=@n12, Alpha=1, ExternalForce="0,0,0") {
    angularDampingOverride=0.8
    angularSpringConstant=5000
    bAngularSpring=true
    bGravityOverrideInSimSpace=true
    bUseGravityOverride=true
    alphaBoolBlend {blendOption=Linear}
    boundBone {boneName=dyn_hair_fr_r_01}
    componentAppliedLinearAccClamp=<10000 10000 10000>
    componentLinearAccScale=<1 1 1>
    componentLinearVelScale=<1 1 1>
    gravityOverride {z=-1200}
    physicsBodyDefinitions [{
      boundBone {boneName=dyn_hair_fr_r_01}
      boxExtents=<9 9 9>
      constraintSetup {angularLimitsMax=<5 20 5>, angularLimitsMin=<0 -50 -20>, angularTarget=<0 0 0>}
      localJointOffset=<0.1 0 0>
     }]
    relativeSpaceBone {boneName=neck_01}
   }
   Anim Dynamics - Bone: dyn_hair_fr_r_02(ComponentPose=@n13, Alpha=1, ExternalForce="0,0,0") {
    angularDampingOverride=0.8
    angularSpringConstant=5000
    bAngularSpring=true
    bGravityOverrideInSimSpace=true
    bUseGravityOverride=true
    alphaBoolBlend {blendOption=Linear}
    boundBone {boneName=dyn_hair_fr_r_02}
    componentAppliedLinearAccClamp=<10000 10000 10000>
    componentLinearAccScale=<1 1 1>
    componentLinearVelScale=<1 1 1>
    gravityOverride {z=-1200}
    physicsBodyDefinitions [{
      boundBone {boneName=dyn_hair_fr_r_02}
      boxExtents=<9 9 9>
      constraintSetup {angularLimitsMax=<5 20 5>, angularLimitsMin=<0 -70 -20>, angularTarget=<0 0 0>}
      localJointOffset=<0.1 0 0>
     }]
    relativeSpaceBone {boneName=neck_01}
   }
   Component To Local()
   "Constraint - dyn_hair_bck_c_02 (P)"(Alpha=1, ConstraintWeights_0=1) {
    alphaBoolBlend {blendOption=Linear}
    boneToModify {boneName=dyn_hair_bck_c_02}
    constraintSetup [{targetBone {boneName=dyn_hair_bck_r_02}}]
    constraintWeights [1]
   }
   Control Rig() {
    alphaBoolBlend {blendOption=Linear}
    controlRigAssetReference {blueprintRigClass=&/F_MED_MillionaireCowgirl_FaceAcc_CtrlRig.F_MED_MillionaireCowgirl_FaceAcc_CtrlRig_C}
   }
   Local To Component()
   "Transform (Modify) Bone - Bone: dyn_hair_bck_l_MSTR"(Alpha=1, Translation="-1.579,0.644,188.485", Rotation="14.965,-0.882,0.633", Scale="1,1,1") {
    rotationMode=BMM_Additive
    rotationSpace=BCS_BoneSpace
    translationMode=BMM_Additive
    translationSpace=BCS_BoneSpace
    alphaBoolBlend {blendOption=Linear}
    boneToModify {boneName=dyn_hair_bck_l_MSTR}
    rotation=<rot 14.965 -0.882 0.633>
    translation=<-1.579 0.644 188.485>
   }
   n14 = Copy Pose From Mesh(SourceMeshComponent=Get Body Mesh() -> (Body Mesh:Skeletal Mesh Component Object Reference)) {bUseAttachedParent=true}
   n15 = Copy Pose From Mesh() {bCopyCurves=true, bUseAttachedParent=true}
   n16 = "Constraint - dyn_hair_bck_c_01 (PP)"!(
    ComponentPose=Local To Component(LocalPose=Layered blend per bone(BasePose=@n15, BlendPoses_0=@n14) {layerSetup [{branchFilters [{boneName=upperarm_l}, {boneName=upperarm_r}]}]})
    Alpha=1
    ConstraintWeights_0=0.3
    ConstraintWeights_1:float=0.7
   ) -> (Pose) {
    alphaBoolBlend {blendOption=Linear}
    boneToModify {boneName=dyn_hair_bck_c_01}
    constraintSetup [{targetBone {boneName=head}}, {targetBone {boneName=spine_05}}]
    constraintWeights [1,1]
   }
   n17 = "Constraint - dyn_hair_bck_l_01 (PP)"!(ComponentPose=@n16, Alpha=1, ConstraintWeights_0=0.3, ConstraintWeights_1:float=0.7) -> (Pose) {
    alphaBoolBlend {blendOption=Linear}
    boneToModify {boneName=dyn_hair_bck_l_01}
    constraintSetup [{targetBone {boneName=head}}, {targetBone {boneName=dyn_hair_bck_c_01}}]
    constraintWeights [1,1]
   }
   n18 = "Constraint - dyn_hair_bck_r_01 (PP)"!(ComponentPose=@n17, Alpha=1, ConstraintWeights_0=0.3, ConstraintWeights_1:float=0.7) -> (Pose) {
    alphaBoolBlend {blendOption=Linear}
    boneToModify {boneName=dyn_hair_bck_r_01}
    constraintSetup [{targetBone {boneName=head}}, {targetBone {boneName=dyn_hair_bck_c_01}}]
    constraintWeights [1,1]
   }
   n19 = "Constraint - dyn_hair_side_l_01 (PP)"!(ComponentPose=@n18, Alpha=1, ConstraintWeights_0=0.3, ConstraintWeights_1:float=0.7) -> (Pose) {
    alphaBoolBlend {blendOption=Linear}
    boneToModify {boneName=dyn_hair_side_l_01}
    constraintSetup [{targetBone {boneName=head}}, {targetBone {boneName=dyn_hair_bck_l_01}}]
    constraintWeights [1,1]
   }
   n20 = "Constraint - dyn_hair_side_r_01 (PP)"!(ComponentPose=@n19, Alpha=1, ConstraintWeights_0=0.3, ConstraintWeights_1:float=0.7) -> (Pose) {
    alphaBoolBlend {blendOption=Linear}
    boneToModify {boneName=dyn_hair_side_r_01}
    constraintSetup [{targetBone {boneName=head}}, {targetBone {boneName=dyn_hair_bck_r_01}}]
    constraintWeights [1,1]
   }
   n21 = Get headup() -> (headup:double)
   n22 = "Transform (Modify) Bone - Bone: dyn_hat_mstr"(ComponentPose=@n20, Alpha=@n21, Translation="-1,0,0", Rotation="12,-0.008,0.004", Scale="1,1,1") {
    rotationMode=BMM_Additive
    rotationSpace=BCS_BoneSpace
    translationMode=BMM_Additive
    translationSpace=BCS_BoneSpace
    alphaBoolBlend {blendOption=Linear}
    boneToModify {boneName=dyn_hat_mstr}
    rotation=<rot 12 -0.008 0.004>
    translation {x=-1}
   }
   n23 = "Transform (Modify) Bone - Bone: dyn_hat_bk"(ComponentPose=@n22, Alpha=@n21, Translation="0,0,0", Rotation="30.000,-0.008,0.004", Scale="1,1,1") {
    rotationMode=BMM_Additive
    rotationSpace=BCS_BoneSpace
    translationMode=BMM_Additive
    translationSpace=BCS_BoneSpace
    alphaBoolBlend {blendOption=Linear}
    boneToModify {boneName=dyn_hat_bk}
    rotation=<rot 29.9999 -0.008 0.004>
   }
   n24 = Get headup()
   n25 = "Transform (Modify) Bone - Bone: dyn_hair_bck_l_MSTR"(ComponentPose=@n23, Alpha=@n24, Translation="-5,2,5", Rotation="35,-0.008,0", Scale="1,1,1") {
    rotationMode=BMM_Additive
    rotationSpace=BCS_BoneSpace
    translationMode=BMM_Additive
    translationSpace=BCS_BoneSpace
    alphaBoolBlend {blendOption=Linear}
    boneToModify {boneName=dyn_hair_bck_l_MSTR}
    rotation {pitch=35, yaw=-0.008}
    translation=<-5 2 5>
   }
   n26 = "Transform (Modify) Bone - Bone: dyn_hair_bck_c_MSTR"(ComponentPose=@n25, Alpha=@n24, Translation="-4,3,4", Rotation="0,0.03,-0.17", Scale="1,1,1") {
    rotationMode=BMM_Additive
    rotationSpace=BCS_BoneSpace
    translationMode=BMM_Additive
    translationSpace=BCS_BoneSpace
    alphaBoolBlend {blendOption=Linear}
    boneToModify {boneName=dyn_hair_bck_c_MSTR}
    rotation {roll=-0.17, yaw=0.03}
    translation=<-4 3 4>
   }
   n27 = "Transform (Modify) Bone - Bone: dyn_hair_bck_r_MSTR"(ComponentPose=@n26, Alpha=@n24, Translation="-4,3,4", Rotation="35,0,0", Scale="1,1,1") {
    rotationMode=BMM_Additive
    rotationSpace=BCS_BoneSpace
    translationMode=BMM_Additive
    translationSpace=BCS_BoneSpace
    alphaBoolBlend {blendOption=Linear}
    boneToModify {boneName=dyn_hair_bck_r_MSTR}
    rotation {pitch=35}
    translation=<-4 3 4>
   }
   n28 = "Transform (Modify) Bone - Bone: dyn_hair_side_r_MSTR"(ComponentPose=@n27, Alpha=@n24, Translation="-2,0,-3", Rotation="-25,35,-0.008", Scale="1,1,1") {
    rotationMode=BMM_Additive
    rotationSpace=BCS_BoneSpace
    translationMode=BMM_Additive
    translationSpace=BCS_BoneSpace
    alphaBoolBlend {blendOption=Linear}
    boneToModify {boneName=dyn_hair_side_r_MSTR}
    rotation=<rot -25 35 -0.008>
    translation {x=-2, z=-3}
   }
   n29 = "Transform (Modify) Bone - Bone: dyn_hair_bck_l_MSTR"!(ComponentPose=@n28, bAlphaBoolEnabled:bool=True, Translation="0,0,0", Rotation="5,0,0", Scale="1,1,1") -> (Pose) {
    alphaInputType=Bool
    rotationMode=BMM_Additive
    rotationSpace=BCS_BoneSpace
    scaleMode=BMM_Additive
    translationMode=BMM_Additive
    translationSpace=BCS_BoneSpace
    alphaBoolBlend {blendOption=Linear}
    boneToModify {boneName=dyn_hair_bck_l_MSTR}
    rotation {pitch=5}
   }
   n30 = "Transform (Modify) Bone - Bone: dyn_hair_fr_l_MSTR"!(ComponentPose=@n29, bAlphaBoolEnabled:bool=Get bIsInFrontEnd(), Translation="-1.5,0,0", Rotation="30,0,0", Scale="0.6,1,1") -> (Pose) {
    alphaInputType=Bool
    rotationMode=BMM_Additive
    rotationSpace=BCS_BoneSpace
    translationMode=BMM_Additive
    translationSpace=BCS_BoneSpace
    alphaBoolBlend {blendOption=Linear}
    boneToModify {boneName=dyn_hair_fr_l_MSTR}
    rotation {pitch=30}
    scale {x=0.6}
    translation {x=-1.5}
   }
   n31 = "Transform (Modify) Bone - Bone: dyn_hair_side_l_MSTR"(ComponentPose=@n30, Alpha=1, Translation="0,0,0", Rotation="40.000,0.029,-0.001", Scale="1,1,1") {
    rotationMode=BMM_Additive
    rotationSpace=BCS_BoneSpace
    translationMode=BMM_Additive
    translationSpace=BCS_BoneSpace
    alphaBoolBlend {blendOption=Linear}
    boneToModify {boneName=dyn_hair_side_l_MSTR}
    rotation=<rot 40 0.029 -0.001>
   }
   n32 = RigidBody!(ComponentPose=@n31, Alpha, OverrideWorldGravity:Vector="0,0,0", ExternalForce) -> (Pose) {
    bClampLinearTranslationLimitToRefPose=true
    bOverrideWorldGravity=true
    overridePhysicsAsset=&/F_MED_MillionaireCowgirl_FaceAcc_Physics
    simulationSpace=BaseBoneSpace
    alphaBoolBlend {blendOption=Linear}
    baseBoneRef {boneName=spine_05}
    componentAppliedLinearAccClamp=<3000 3000 3000>
    componentLinearAccScale=<0.35 0.7 0.7>
    componentLinearVelScale=<0.35 0.7 0.7>
   }
   n33 = "Transform (Modify) Bone - Bone: dyn_hair_side_r_MSTR"(ComponentPose=@n32, Alpha=1, Translation="0,0,0", Rotation="0,0,0", Scale="0.9,1,1") {
    rotationMode=BMM_Additive
    rotationSpace=BCS_BoneSpace
    scaleMode=BMM_Replace
    translationMode=BMM_Additive
    translationSpace=BCS_BoneSpace
    alphaBoolBlend {blendOption=Linear}
    boneToModify {boneName=dyn_hair_side_r_MSTR}
    scale {x=0.9}
   }
   n34 = "Transform (Modify) Bone - Bone: dyn_hair_side_r_02"(ComponentPose=@n33, Alpha=1, Translation="0,0,1", Rotation="-20.541,-1.800,-0.549", Scale="0.9,1,1") {
    rotationMode=BMM_Additive
    rotationSpace=BCS_BoneSpace
    scaleMode=BMM_Replace
    translationMode=BMM_Additive
    translationSpace=BCS_BoneSpace
    alphaBoolBlend {blendOption=Linear}
    boneToModify {boneName=dyn_hair_side_r_02}
    rotation=<rot -20.5411 -1.8001 -0.5489>
    scale {x=0.9}
    translation {z=1}
   }
   n35 = "Constraint - dyn_hair_side_r_02 (P)"(ComponentPose=@n34, Alpha=0.1, ConstraintWeights_0=1) {
    alphaBoolBlend {blendOption=Linear}
    boneToModify {boneName=dyn_hair_side_r_02}
    constraintSetup [{targetBone {boneName=dyn_hair_bck_r_01}}]
    constraintWeights [1]
   }
   n36 = "Constraint - dyn_hair_bck_r_02 (PP)"!(ComponentPose=@n35, Alpha=0.2, ConstraintWeights_0=0.8, ConstraintWeights_1:float=0.2) -> (Pose) {
    alphaBoolBlend {blendOption=Linear}
    boneToModify {boneName=dyn_hair_bck_r_02}
    constraintSetup [{targetBone {boneName=dyn_hair_side_r_02}}, {targetBone {boneName=dyn_hair_bck_c_02}}]
    constraintWeights [1,1]
   }
   n37 = Control Rig(Source=Component To Local(ComponentPose=@n36)) {
    alphaBoolBlend {blendOption=Linear}
    controlRigAssetReference {blueprintRigClass=&F_MED_MillionaireCowgirl_FaceAcc_CtrlRig_C}
   }
   Output Pose(Result=@n37) {layerGroup=None}
  ]}
 EventGraph {nodes [
   Event BlueprintUpdateAnimation()
   > n1 = Gravity Override State Function DEPRECATED(CharacterPartAnimInstance=Self-Reference(), GameTimeInSeconds=GetGameTimeInSeconds(), GravityOverrideParams=Get Gravity Override Params())
   > Set Gravity Override(Gravity Override=@n1)
   > Set Arm R rot(Arm R rot:double=GetAlphaFromReferenceTransformDelta(SkeletalMeshComponent=Get Skel_Mesh(), BoneName=upperarm_r, BaseName=spine_05, InputRange=(X=40,Y=120), TwistAxis="1,0,0")) -> (Output_Get:double)
   > Set headup(headup:double=GetAlphaFromReferenceTransformDelta(SkeletalMeshComponent=Get Skel_Mesh(), BoneName=head, BaseName=spine_05, InputRange=(X=-40,Y=-70), TwistAxis="1,0,0")) -> (Output_Get:double)
   Event BlueprintInitializeAnimation()
   > Set Skel_Mesh(Skel_Mesh=GetOwnerSkeletalMeshComponent())
  ]}
]
variables [
 Arm R rot {type=number}
 Body Mesh {type=SkeletalMeshComponent}
 Gravity Override {type=Vector}
 Gravity Override Params {type=GravityOverrideParamsStruct}
 headup {type=number}
 Skel_Mesh {type=SkeletalMeshComponent}
]
